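# Material function that layers EchoFrost icy puddle patches over incoming material attributes, toggled by UseCoverIce and bypassed for material proxies.
# It adjusts roughness, specular, metallic and base color, and flattens normals toward world up.
# Named masks derive from world-XY puddle noise, slope, height, roughness-based porosity, UV1 layer indices and an object-space gradient, also output separately.
# The final blend uses a contrasted mask times inverted CoverMask; puddle roughness varies by quality level.
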MaterialFunction /Hera/Sets/EchoFrost/Materials/MaterialFunctions/MF_EchoFrost_IcePatches
graph {nodes [
  n1 = NamedRerouteUsage()
  Ceil(None=@n1)
  n2 = VectorParameter() {
   group=Cover
   parameterName=Spec Tex_Channel Mix
   sortPriority=61
   channelNames {a="", b=Roughness, g=Metallic, r=Specular}
   defaultValue=<rgba 0 1 0 0>
  }
  DotProduct(A=@n2.RGB)
  n3 = ScalarParameter() {group=EchoFrost Cover Ice Patches, parameterName=Puddle_Roughness}
  n4 = ScalarParameter() {group=EchoFrost Cover Ice Patches, parameterName="Puddle_Roughness (Med/Low)"}
  n5 = QualitySwitch(Default=@n3, Low=@n4, High=@n3, Medium=@n4, Epic=@n3)
  n6 = NamedRerouteUsage()
  n7 = FunctionInput() {inputType=FunctionInput_MaterialAttributes}
  n8 = GetMaterialAttributes(None=@n7) {attributeGetTypes [
    #1
    #2
    #3
    #4
    #5
   ]}
  n9 = Multiply(A=ScalarParameter() {defaultValue=1, group=EchoFrost Cover Ice Patches, parameterName=Puddle_Specular}, B=@n8.Specular)
  n10 = NamedRerouteUsage()
  n11 = Constant3Vector() {constant=<rgba 0 0 1 0>}
  n12 = ScalarParameter() {defaultValue=0.6, group=EchoFrost Cover Ice Patches, parameterName=PuddleWet_Darken}
  n13 = NamedRerouteUsage()
  n14 = SetMaterialAttributes!(
   MaterialAttributes=@n8.MaterialAttributes
   Roughness=Min(
    A=LinearInterpolate(A=@n8.Roughness, B=@n5, Alpha=@n6)
    B=ScalarParameter() {defaultValue=0.32, group=EchoFrost Cover Ice Patches, parameterName=Puddle_Roughness_min}
   )
   Specular=Min(A=LinearInterpolate(A=@n8.Specular, B=@n9, Alpha=@n10)) {constB=0.5}
   Normal=LinearInterpolate(A=@n8.Normal, B=Transform(None=@n11) {transformSourceType=TRANSFORMSOURCE_World, transformType=TRANSFORM_Tangent}, Alpha=NamedRerouteUsage())
   BaseColor=Multiply(A=@n8.BaseColor, B=LinearInterpolate(B=@n12, Alpha=@n13) {constA=1})
   Metallic=LinearInterpolate(
    A=@n8.Metallic
    B=ScalarParameter() {defaultValue=0.1, group=EchoFrost Cover Ice Patches, parameterName=Puddle_Metallic}
    Alpha=NamedRerouteUsage()
   )
  ) {attributeSetTypes [
    #1
    #2
    #3
    #4
    #5
   ]}
  n15 = NamedRerouteUsage()
  n16 = ScalarParameter() {defaultValue=5, group=EchoFrost Cover Ice Patches Masking, parameterName=Contrast}
  n17 = FunctionInput() {bUsePreviewValueAsDefault=true, inputName=CoverMask, inputType=FunctionInput_Scalar, sortPriority=50}
  n18 = BlendMaterialAttributes(
   A=@n7
   B=@n14
   Alpha=Multiply(
    A=MaterialFunctionCall(In=@n15, Contrast=@n16) {materialFunction=&CheapContrast}.Result
    B=OneMinus(None=@n17)
   )
  )
  FunctionOutput(None=MaterialProxyReplace(Realtime=StaticSwitchParameter(True=@n18, False=@n7) {group=Cover, parameterName=UseCoverIce, sortPriority=-5}, MaterialProxy=@n7))
  n19 = VectorParameter() {group=EchoFrost Cover Ice Patches Masking, parameterName=IceDry_Gradient Direction, defaultValue=<rgba 0 1 0 1>}
  n20 = DotProduct(A=Transform(None=@n19.RGB) {transformSourceType=TRANSFORMSOURCE_Local}, B=Subtract(A=WorldPosition().XYZ, B=ObjectPositionWS()))
  n21 = Multiply(A=ScalarParameter() {defaultValue=1, group=EchoFrost Cover Ice Patches Masking, parameterName=IceDry_GradientLength}, B=ObjectRadius())
  n22 = Divide(A=@n21)
  n23 = Power(
   Base=Divide(A=Add(A=@n20, B=@n22), B=@n21)
   Exp=ScalarParameter() {defaultValue=1, group=EchoFrost Cover Ice Patches Masking, parameterName=IceDry_Gradient_Falloff}
  )
  FunctionOutput(None=NamedRerouteDeclaration(None=Saturate(None=@n23)) {name=WetDryMask, nodeColor=<rgba 1 0 0.1172 1>}) {outputName=DryWetMaskTransition, sortPriority=10}
  n24 = Multiply(A=NamedRerouteUsage(), B=ScalarParameter() {defaultValue=0.1, group=EchoFrost Cover Ice Patches, parameterName=Ice_HeightMap_influence})
  n25 = LinearInterpolate(
   A=ScalarParameter() {defaultValue=0.4608, group=EchoFrost Cover Ice Patches, parameterName=Input_Roughness_Porosity_Low}
   B=ScalarParameter() {defaultValue=1, group=EchoFrost Cover Ice Patches, parameterName=Input_Roughness_Porosity_High}
   Alpha=@n8.Roughness
  )
  n26 = LinearInterpolate(
   A=@n24
   B=Saturate(None=@n25)
   Alpha=ScalarParameter() {defaultValue=1, group=EchoFrost Cover Ice Patches, parameterName=Ice_Porosity_Influence}
  )
  NamedRerouteDeclaration(None=@n26) {name=Porosity, nodeColor=<rgba 0 1 0.4844 1>}
  n27 = Round(None=@n1)
  n28 = ScalarParameter() {defaultValue=0.752, group=EchoFrost Cover Ice Patches, parameterName=Offset PuddleByHeight}
  n29 = ComponentMask(None=WorldPosition().XYZ) {g=true, r=true}
  n30 = ScalarParameter() {defaultValue=2000, group=EchoFrost Cover Ice Patches, parameterName=IcePuddle_Size}
  n31 = TextureSample(UVs=Add(A=Multiply(A=@n27, B=@n28), B=Divide(A=@n29, B=@n30))) {
   samplerSource=SSM_Wrap_WorldGroupSettings
   samplerType=SAMPLERTYPE_LinearColor
   texture=/Game/Environments/Asteria/Sets/NeonCity/Textures/WaterPuddles_M
  }
  n32 = Saturate(None=LinearInterpolate(
    A=ScalarParameter() {defaultValue=-6.4879, group=EchoFrost Cover Ice Patches, parameterName=IcePuddle_Contrast_Low}
    B=ScalarParameter() {defaultValue=17.5196, group=EchoFrost Cover Ice Patches, parameterName=IcePuddle_Contrast_High}
    Alpha=@n31.R
   ))
  n33 = Multiply(A=NamedRerouteUsage(), B=ScalarParameter() {defaultValue=1, group=EchoFrost Cover Ice Patches, parameterName=IcePuddle_HeightMap_influence})
  n34 = ScalarParameter() {defaultValue=-78.5354, group=EchoFrost Cover Ice Patches, parameterName=IcePuddle_SlopeMask_Low}
  n35 = ScalarParameter() {defaultValue=1.2891, group=EchoFrost Cover Ice Patches, parameterName=IcePuddle_SlopeMask_High}
  n36 = ComponentMask(None=VertexNormalWS()) {b=true}
  n37 = Multiply(A=Saturate(None=Subtract(A=@n32, B=@n33)), B=Saturate(None=LinearInterpolate(A=@n34, B=@n35, Alpha=@n36)))
  NamedRerouteDeclaration(None=@n37) {name=PuddleMask, nodeColor=<rgba 1 0.8438 0 1>}
  NamedRerouteDeclaration(None=OneMinus(None=FunctionInput() {inputName=Height, inputType=FunctionInput_Scalar, sortPriority=10})) {name=Height, nodeColor=<rgba 1 0 0.1875 1>}
  n38 = Subtract(A=NamedRerouteUsage())
  n39 = Multiply(A=@n37) {constB=2}
  NamedRerouteDeclaration(None=Saturate(None=Add(A=@n38, B=@n39))) {name=WetMask, nodeColor=<rgba 0 1 0.4844 1>}
  n40 = Constant() {r=3}
  n41 = ComponentMask(None=TextureCoordinate() {coordinateIndex=1}) {r=true}
  n42 = Floor(None=@n41)
  n43 = Abs(None=Subtract(A=@n42, B=Constant() {r=2}))
  n44 = Abs(None=Subtract(A=@n42, B=Constant() {r=1}))
  n45 = Constant() {r=1}
  n46 = StaticSwitch(
   True=If(A=@n41, B=Constant() {r=1}, "A > B"=@n45, "A < B"=Constant()) {desc=0}
   False=@n45
   Value=StaticBoolParameter() {group=EchoFrost Cover Ice Patches Masking, parameterName=Wet_Maskout_Layer_1}
  )
  n47 = StaticSwitch(
   True=If(A=@n41, B=Constant() {r=1}, "A > B"=Saturate(None=@n44), "A < B"=@n46) {desc=0}
   False=@n46
   Value=StaticBoolParameter() {group=EchoFrost Cover Ice Patches Masking, parameterName=Wet_Maskout_Layer_2}
  )
  n48 = StaticSwitch(
   True=If(A=@n41, B=Constant() {r=2}, "A > B"=Saturate(None=@n43), "A < B"=@n47)
   False=@n47
   Value=StaticBoolParameter() {group=EchoFrost Cover Ice Patches Masking, parameterName=Wet_Maskout_Layer_3}
  )
  n49 = Constant() {r=3}
  n50 = Saturate(None=Abs(None=Subtract(A=@n42, B=@n49)))
  NamedRerouteDeclaration(None=StaticSwitch(
    True=If(A=@n41, B=@n40, "A > B"=@n50, "A < B"=@n48)
    False=@n48
    Value=StaticBoolParameter() {group=EchoFrost Cover Ice Patches Masking, parameterName=Wet_Maskout_Layer_4}
   )) {name=WetLayerMask, nodeColor=<rgba 0.8047 0 1 1>}
  NamedRerouteUsage()
  NamedRerouteUsage()
  Saturate(None=@n26)
  Saturate(None=Multiply(A=@n17, B=ScalarParameter() {defaultValue=1, group=EchoFrost Cover Ice Patches Masking, parameterName=CoverMask Multiplier}) {constB=2})
  n51 = GetMaterialAttributes(None=@n7) {attributeGetTypes [#1]}
  n52 = Add(A=@n51.Roughness, B=ScalarParameter() {group=EchoFrost Cover Ice Patches, parameterName="Other Layers Roughness Bias (Med/Low)"})
  SetMaterialAttributes!(
   MaterialAttributes=@n51.MaterialAttributes
   Roughness=QualitySwitch(Default=@n51.Roughness, Low=@n52, High=@n51.Roughness, Medium=@n52, Epic=@n51.Roughness)
  ) {attributeSetTypes [#1]}
  Subtract(A=Saturate(None=@n25), B=@n24)
 ]}
