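# Material function that builds the EchoFrost snow layer's material attributes, with time-of-day color and parameter-driven roughness and specular.
# A snow normal sampled with supplied world-space UVs and derivatives is triplanar-converted, then angle-corrected blended over the incoming normal.
# Sparkle emissive combines reflection-vector masks, skylight samples and distance fade, using depth-based ping-pong UV scales to crossfade two texture scales.
# Mobile uses flat color and the input normal without sparkle; Low quality also drops sparkle.
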MaterialFunction /Hera/Sets/EchoFrost/Materials/MaterialFunctions/MF_EchoFrost_Snow
graph {expressions {
  MaterialExpressionAbs_1 {class=MaterialExpressionAbs, inputs [None {from=MaterialExpressionAdd_10, output=""}], properties {}}
  MaterialExpressionAdd_10 {
   class=MaterialExpressionAdd
   inputs [
    A {from=MaterialExpressionFrac_2}
    B {from=""}
   ]
   properties {constB=-0.5}
  }
  MaterialExpressionAdd_7 {
   class=MaterialExpressionAdd
   inputs [
    A {from=MaterialExpressionComponentMask_2}
    B {from=MaterialExpressionConstant3Vector_1}
   ]
  }
  MaterialExpressionAdd_9 {
   class=MaterialExpressionAdd
   inputs [
    A {from=MaterialExpressionDivide_6}
    B {from=""}
   ]
   properties {constB=0.5}
  }
  MaterialExpressionCameraVectorWS_3 {class=MaterialExpressionCameraVectorWS, inputs [], properties {}}
  MaterialExpressionCeil_1 {class=MaterialExpressionCeil, inputs [None {from=MaterialExpressionDivide_6, output=""}], properties {}}
  MaterialExpressionCeil_2 {class=MaterialExpressionCeil, inputs [None {from=MaterialExpressionAdd_9, output=""}], properties {}}
  MaterialExpressionComponentMask_2 {class=MaterialExpressionComponentMask, inputs [None {from=MaterialExpressionLinearInterpolate_4}], properties {b=true, g=true, r=true}}
  MaterialExpressionComponentMask_6 {class=MaterialExpressionComponentMask, inputs [None {from=MaterialExpressionSaturate_13}], properties {r=true}}
  MaterialExpressionComponentMask_8 {class=MaterialExpressionComponentMask, inputs [None {from=MaterialExpressionLinearInterpolate_10}], properties {g=true}}
  MaterialExpressionConstant3Vector_0 {class=MaterialExpressionConstant3Vector, properties {constant=<rgba 0 0 1 0>}}
  MaterialExpressionConstant3Vector_1 {class=MaterialExpressionConstant3Vector, properties {constant=<rgba 0 0 -0.9 1>}}
  MaterialExpressionConstant_0 {class=MaterialExpressionConstant, properties {r=30}}
  MaterialExpressionConstant_1 {class=MaterialExpressionConstant, properties {r=75}}
  MaterialExpressionConstant_11 {class=MaterialExpressionConstant}
  MaterialExpressionConstant_2 {class=MaterialExpressionConstant, properties {r=1}}
  MaterialExpressionConstant_5 {class=MaterialExpressionConstant, properties {r=0.5}}
  MaterialExpressionConvert_1 {
   class=MaterialExpressionConvert
   inputs [
    None {from=MaterialExpressionNamedRerouteUsage_90}
    None {from=MaterialExpressionNamedRerouteUsage_91}
    None {from=MaterialExpressionNamedRerouteUsage_92}
   ]
   properties {
    nodeName=Make Vector3
    convertInputs [
     {
      type=Scalar
      expressionInput {
       expression=&self:MaterialExpressionNamedRerouteUsage_90
       inputName=None
       mask=0
       maskA=0
       maskB=0
       maskG=0
       maskR=0
       outputIndex=0
      }
     }
     {
      type=Scalar
      expressionInput {
       expression=&self:MaterialExpressionNamedRerouteUsage_91
       inputName=None
       mask=0
       maskA=0
       maskB=0
       maskG=0
       maskR=0
       outputIndex=0
      }
     }
     {
      type=Scalar
      expressionInput {
       expression=&self:MaterialExpressionNamedRerouteUsage_92
       inputName=None
       mask=0
       maskA=0
       maskB=0
       maskG=0
       maskR=0
       outputIndex=0
      }
     }
    ]
    convertOutputs [{type=Vector3}]
   }
  }
  MaterialExpressionDivide_1 {
   class=MaterialExpressionDivide
   inputs [
    A {from=MaterialExpressionPixelDepth_0, output=R}
    B {from=MaterialExpressionScalarParameter_27}
   ]
   properties {constB=5000}
  }
  MaterialExpressionDivide_6 {
   class=MaterialExpressionDivide
   inputs [
    A {from=MaterialExpressionPixelDepth_2, output=R}
    B {from=MaterialExpressionScalarParameter_17}
   ]
   properties {constB=5000}
  }
  MaterialExpressionDivide_7 {
   class=MaterialExpressionDivide
   inputs [
    A {from=""}
    B {from=MaterialExpressionCeil_1}
   ]
  }
  MaterialExpressionDivide_8 {
   class=MaterialExpressionDivide
   inputs [
    A {from=""}
    B {from=MaterialExpressionCeil_2}
   ]
  }
  MaterialExpressionDotProduct_1 {
   class=MaterialExpressionDotProduct
   inputs [
    A {from=MaterialExpressionReflectionVectorWS_2, output=""}
    B {from=MaterialExpressionCameraVectorWS_3, output=""}
   ]
   properties {}
  }
  MaterialExpressionEyeAdaptationInverse_0 {
   class=MaterialExpressionEyeAdaptationInverse
   inputs [
    LightValueInput {from=MaterialExpressionNamedRerouteUsage_30, output=""}
    AlphaInput {from=MaterialExpressionConstant_2, output=""}
   ]
   properties {}
  }
  MaterialExpressionFrac_2 {class=MaterialExpressionFrac, inputs [None {from=MaterialExpressionDivide_6, output=""}], properties {}}
  MaterialExpressionFunctionInput_1 {
   class=MaterialExpressionFunctionInput
   inputs [Preview {from="", output=""}]
   properties {bUsePreviewValueAsDefault=true, inputName=IceSnow WS - UVs, inputType=FunctionInput_Vector2, sortPriority=900}
  }
  MaterialExpressionFunctionInput_10 {
   class=MaterialExpressionFunctionInput
   inputs [Preview {from="", output=""}]
   properties {inputName=DFAO, inputType=FunctionInput_Scalar, sortPriority=813}
  }
  MaterialExpressionFunctionInput_11 {
   class=MaterialExpressionFunctionInput
   inputs [Preview {from=MaterialExpressionSetMaterialAttributes_0, output=""}]
   properties {
    description=MaterialAttributes from previous layers
    inputName=Input MaterialAttributes
    inputType=FunctionInput_MaterialAttributes
    sortPriority=-100
   }
  }
  MaterialExpressionFunctionInput_12 {
   class=MaterialExpressionFunctionInput
   inputs [Preview {from="", output=""}]
   properties {inputName=Baked Curvature, inputType=FunctionInput_Scalar, sortPriority=810}
  }
  MaterialExpressionFunctionInput_14 {
   class=MaterialExpressionFunctionInput
   inputs [Preview {from="", output=""}]
   properties {inputName=Z Gradient, inputType=FunctionInput_Scalar, sortPriority=814}
  }
  MaterialExpressionFunctionInput_15 {
   class=MaterialExpressionFunctionInput
   inputs [Preview {from="", output=""}]
   properties {inputName=Baked AO, inputType=FunctionInput_Scalar, sortPriority=811}
  }
  MaterialExpressionFunctionInput_16 {
   class=MaterialExpressionFunctionInput
   inputs [Preview {from="", output=""}]
   properties {inputName=Painted Mask, inputType=FunctionInput_Scalar, sortPriority=812}
  }
  MaterialExpressionFunctionInput_2 {
   class=MaterialExpressionFunctionInput
   inputs [Preview {from="", output=""}]
   properties {bUsePreviewValueAsDefault=true, inputName=IceSnow WS - DDX, inputType=FunctionInput_Vector2, sortPriority=901}
  }
  MaterialExpressionFunctionInput_26 {class=MaterialExpressionFunctionInput, inputs [Preview {from="", output=""}], properties {inputName=WS Noise Masks, sortPriority=815}}
  MaterialExpressionFunctionInput_3 {
   class=MaterialExpressionFunctionInput
   inputs [Preview {from="", output=""}]
   properties {bUsePreviewValueAsDefault=true, inputName=IceSnow WS - DDY, inputType=FunctionInput_Vector2, sortPriority=902}
  }
  MaterialExpressionFunctionInput_4 {
   class=MaterialExpressionFunctionInput
   inputs [Preview {from="", output=""}]
   properties {bUsePreviewValueAsDefault=true, inputName=IceSnow WS - Mask, sortPriority=903}
  }
  MaterialExpressionFunctionInput_8 {
   class=MaterialExpressionFunctionInput
   inputs [Preview {from="", output=""}]
   properties {inputName="\\ --- Dependencies --- /", inputType=FunctionInput_Scalar, sortPriority=800}
  }
  MaterialExpressionFunctionOutput_0 {
   class=MaterialExpressionFunctionOutput
   inputs [None {from=MaterialExpressionNamedRerouteDeclaration_3, output=""}]
   properties {outputName=Snow MaterialAttributes, sortPriority=-100}
  }
  MaterialExpressionGetMaterialAttributes_0 {
   class=MaterialExpressionGetMaterialAttributes
   inputs [None {from=MaterialExpressionNamedRerouteUsage_0, output=""}]
   properties {attributeGetTypes [#1]}
  }
  MaterialExpressionLinearInterpolate_0 {
   class=MaterialExpressionLinearInterpolate
   inputs [
    A {from=MaterialExpressionGetMaterialAttributes_0, output=Normal}
    B {from=MaterialExpressionMaterialFunctionCall_0, output=Result}
    Alpha {from=MaterialExpressionScalarParameter_1}
   ]
  }
  MaterialExpressionLinearInterpolate_10 {
   class=MaterialExpressionLinearInterpolate
   inputs [
    A {from=MaterialExpressionTextureSample_4, output=RGB}
    B {from=MaterialExpressionTextureSample_6, output=RGB}
    Alpha {from=MaterialExpressionNamedRerouteUsage_27}
   ]
  }
  MaterialExpressionLinearInterpolate_4 {
   class=MaterialExpressionLinearInterpolate
   inputs [
    A {from=MaterialExpressionTextureSample_2, output=RGB}
    B {from=MaterialExpressionTextureSample_5, output=RGB}
    Alpha {from=MaterialExpressionNamedRerouteUsage_18}
   ]
  }
  MaterialExpressionMaterialFunctionCall_0 {
   class=MaterialExpressionMaterialFunctionCall
   inputs [
    BaseNormal {from=MaterialExpressionGetMaterialAttributes_0, output=Normal}
    AdditionalNormal {from=MaterialExpressionTransform_5}
   ]
   properties {materialFunction=&BlendAngleCorrectedNormals}
  }
  MaterialExpressionMaterialFunctionCall_1 {
   class=MaterialExpressionMaterialFunctionCall
   inputs [
    Color_Morning {from=MaterialExpressionVectorParameter_3, output=RGB}
    Color_Day {from=MaterialExpressionVectorParameter_25, output=RGB}
    Color_Evening {from=MaterialExpressionVectorParameter_3, output=RGB}
    Color_Night {from=MaterialExpressionVectorParameter_3, output=RGB}
   ]
   properties {materialFunction=&Color_Time_of_Day}
  }
  MaterialExpressionMaterialFunctionCall_17 {
   class=MaterialExpressionMaterialFunctionCall
   inputs [
    Tangent Normals {from=MaterialExpressionTextureSampleParameter2D_0, output=RGB}
    Vertex Normals {from=""}
    Triplanar Mask {from=MaterialExpressionFunctionInput_4}
   ]
   properties {materialFunction=/Game/Athena/Environments/MaterialFunctions/MF_TriplanarNormalTangentToWorld}
  }
  MaterialExpressionMaterialFunctionCall_2 {
   class=MaterialExpressionMaterialFunctionCall
   inputs [
    Low {from=MaterialExpressionReroute_1}
    High {from=MaterialExpressionReroute_7}
    Epic {from=MaterialExpressionReroute_7}
    -------- {from=MaterialExpressionReroute_7}
    Mobile {from=MaterialExpressionReroute_1}
    Medium {from=MaterialExpressionReroute_7}
   ]
   properties {materialFunction=/FortniteBaseMaterials/Functions/MF_QualitySwitch_Scalar}
  }
  MaterialExpressionMaterialFunctionCall_24 {
   class=MaterialExpressionMaterialFunctionCall
   inputs [
    Absolute World Position {from=""}
    "Normal (WS)" {from=""}
    Texture Scale {from=MaterialExpressionScalarParameter_0}
    Triplanar Fade {from=MaterialExpressionMaterialFunctionCall_2, output=Result}
    Derivatives Scale {from=""}
   ]
   properties {materialFunction=/BRCoreMaterials/Materials/MaterialFunctions/MF_TriplanarDithered_WavyDistortion}
  }
  MaterialExpressionMultiply_0 {
   class=MaterialExpressionMultiply
   inputs [
    A {from=""}
    B {from=""}
   ]
  }
  MaterialExpressionMultiply_10 {
   class=MaterialExpressionMultiply
   inputs [
    A {from=MaterialExpressionSkyLightEnvMapSample_0}
    B {from=MaterialExpressionNamedRerouteUsage_26}
   ]
  }
  MaterialExpressionMultiply_11 {
   class=MaterialExpressionMultiply
   inputs [
    A {from=MaterialExpressionMultiply_10}
    B {from=MaterialExpressionScalarParameter_13}
   ]
  }
  MaterialExpressionMultiply_12 {
   class=MaterialExpressionMultiply
   inputs [
    A {from=MaterialExpressionAbs_1}
    B {from=""}
   ]
   properties {constB=2}
  }
  MaterialExpressionMultiply_16 {
   class=MaterialExpressionMultiply
   inputs [
    A {from=MaterialExpressionNamedRerouteUsage_28}
    B {from=MaterialExpressionDivide_7}
   ]
  }
  MaterialExpressionMultiply_21 {
   class=MaterialExpressionMultiply
   inputs [
    A {from=MaterialExpressionNamedRerouteUsage_29}
    B {from=MaterialExpressionDivide_8}
   ]
  }
  MaterialExpressionMultiply_27 {
   class=MaterialExpressionMultiply
   inputs [
    A {from=MaterialExpressionLinearInterpolate_10}
    B {from=MaterialExpressionLinearInterpolate_10}
   ]
   properties {constB=3}
  }
  MaterialExpressionMultiply_29 {
   class=MaterialExpressionMultiply
   inputs [
    A {from=MaterialExpressionTextureCoordinate_0}
    B {from=MaterialExpressionScalarParameter_25}
   ]
  }
  MaterialExpressionMultiply_3 {
   class=MaterialExpressionMultiply
   inputs [
    A {from=MaterialExpressionDotProduct_1}
    B {from=MaterialExpressionComponentMask_6}
   ]
  }
  MaterialExpressionMultiply_30 {
   class=MaterialExpressionMultiply
   inputs [
    A {from=MaterialExpressionOneMinus_2}
    B {from=MaterialExpressionMultiply_11}
   ]
  }
  MaterialExpressionNamedRerouteDeclaration_10 {
   class=MaterialExpressionNamedRerouteDeclaration
   inputs [None {from=MaterialExpressionMultiply_30}]
   properties {name=Epic_SnowSparkleEmissive, nodeColor=<rgba 0 0.2734 1 1>}
  }
  MaterialExpressionNamedRerouteDeclaration_11 {
   class=MaterialExpressionNamedRerouteDeclaration
   inputs [None {from=MaterialExpressionFunctionInput_10}]
   properties {name=_DFAO, nodeColor=<rgba 1 0 0.0525 1>}
  }
  MaterialExpressionNamedRerouteDeclaration_12 {
   class=MaterialExpressionNamedRerouteDeclaration
   inputs [None {from=MaterialExpressionFunctionInput_12}]
   properties {name=_Baked Curvature, nodeColor=<rgba 1 0 0.0525 1>}
  }
  MaterialExpressionNamedRerouteDeclaration_13 {
   class=MaterialExpressionNamedRerouteDeclaration
   inputs [None {from=MaterialExpressionFunctionInput_16}]
   properties {name=_Painted Mask, nodeColor=<rgba 1 0 0.0525 1>}
  }
  MaterialExpressionNamedRerouteDeclaration_14 {
   class=MaterialExpressionNamedRerouteDeclaration
   inputs [None {from=MaterialExpressionSaturate_15}]
   properties {name=pingPongMask, nodeColor=<rgba 0 1 0.2969 1>}
  }
  MaterialExpressionNamedRerouteDeclaration_15 {
   class=MaterialExpressionNamedRerouteDeclaration
   inputs [None {from=MaterialExpressionMultiply_29}]
   properties {name=SparkleCoords, nodeColor=<rgba 1 0 0.2812 1>}
  }
  MaterialExpressionNamedRerouteDeclaration_16 {
   class=MaterialExpressionNamedRerouteDeclaration
   inputs [None {from=MaterialExpressionComponentMask_8}]
   properties {name=SparkleSoftSpecular, nodeColor=<rgba 0 0.9531 1 1>}
  }
  MaterialExpressionNamedRerouteDeclaration_22 {
   class=MaterialExpressionNamedRerouteDeclaration
   inputs [None {from=MaterialExpressionFunctionInput_26}]
   properties {name=_WS Noise Masks, nodeColor=<rgba 1 0 0.0525 1>}
  }
  MaterialExpressionNamedRerouteDeclaration_23 {
   class=MaterialExpressionNamedRerouteDeclaration
   inputs [None {from=MaterialExpressionConvert_1}]
   properties {name=Baked AO/Curvature/Painted Texture, nodeColor=<rgba 1 1 1 1>}
  }
  MaterialExpressionNamedRerouteDeclaration_3 {
   class=MaterialExpressionNamedRerouteDeclaration
   inputs [None {from=MaterialExpressionSetMaterialAttributes_3}]
   properties {name=Snow - MaterialAttributes, nodeColor=<rgba 0.8333 1 0.9688 1>}
  }
  MaterialExpressionNamedRerouteDeclaration_5 {
   class=MaterialExpressionNamedRerouteDeclaration
   inputs [None {from=MaterialExpressionPower_0}]
   properties {name=HighendSparkleMask, nodeColor=<rgba 0 0.1094 1 1>}
  }
  MaterialExpressionNamedRerouteDeclaration_6 {
   class=MaterialExpressionNamedRerouteDeclaration
   inputs [None {from=MaterialExpressionTransform_1}]
   properties {name=sparklenormal, nodeColor=<rgba 1 0.1875 0 1>}
  }
  MaterialExpressionNamedRerouteDeclaration_7 {
   class=MaterialExpressionNamedRerouteDeclaration
   inputs [None {from=MaterialExpressionFunctionInput_14}]
   properties {name=_Z Gradient, nodeColor=<rgba 1 0 0.0525 1>}
  }
  MaterialExpressionNamedRerouteDeclaration_8 {
   class=MaterialExpressionNamedRerouteDeclaration
   inputs [None {from=MaterialExpressionFunctionInput_15}]
   properties {name=_Baked AO, nodeColor=<rgba 1 0 0.0525 1>}
  }
  MaterialExpressionNamedRerouteDeclaration_9 {
   class=MaterialExpressionNamedRerouteDeclaration
   inputs [None {from=MaterialExpressionReroute_15}]
   properties {name=_Input A MaterialAttributes, desc=MaterialAttributes from previous layers, nodeColor=<rgba 0.1016 0 1 1>}
  }
  MaterialExpressionNamedRerouteUsage_0 {class=MaterialExpressionNamedRerouteUsage}
  MaterialExpressionNamedRerouteUsage_18 {class=MaterialExpressionNamedRerouteUsage}
  MaterialExpressionNamedRerouteUsage_25 {class=MaterialExpressionNamedRerouteUsage}
  MaterialExpressionNamedRerouteUsage_26 {class=MaterialExpressionNamedRerouteUsage}
  MaterialExpressionNamedRerouteUsage_27 {class=MaterialExpressionNamedRerouteUsage}
  MaterialExpressionNamedRerouteUsage_28 {class=MaterialExpressionNamedRerouteUsage}
  MaterialExpressionNamedRerouteUsage_29 {class=MaterialExpressionNamedRerouteUsage}
  MaterialExpressionNamedRerouteUsage_30 {class=MaterialExpressionNamedRerouteUsage}
  MaterialExpressionNamedRerouteUsage_90 {class=MaterialExpressionNamedRerouteUsage}
  MaterialExpressionNamedRerouteUsage_91 {class=MaterialExpressionNamedRerouteUsage}
  MaterialExpressionNamedRerouteUsage_92 {class=MaterialExpressionNamedRerouteUsage}
  MaterialExpressionNormalize_2 {class=MaterialExpressionNormalize, inputs [VectorInput {from=MaterialExpressionAdd_7, output=""}]}
  MaterialExpressionOneMinus_2 {class=MaterialExpressionOneMinus, inputs [None {from=MaterialExpressionSaturate_16}]}
  MaterialExpressionPixelDepth_0 {class=MaterialExpressionPixelDepth, inputs [], properties {}}
  MaterialExpressionPixelDepth_2 {class=MaterialExpressionPixelDepth, inputs [], properties {}}
  MaterialExpressionPower_0 {
   class=MaterialExpressionPower
   inputs [
    Base {from=MaterialExpressionMultiply_3}
    Exp {from=MaterialExpressionScalarParameter_9}
   ]
  }
  MaterialExpressionQualitySwitch_2 {
   class=MaterialExpressionQualitySwitch
   inputs [
    Default {from=MaterialExpressionEyeAdaptationInverse_0, output=EyeAdaptationInverse}
    Low {from=MaterialExpressionConstant_11, output=""}
    High {from=MaterialExpressionEyeAdaptationInverse_0, output=EyeAdaptationInverse}
    Medium {from=MaterialExpressionEyeAdaptationInverse_0, output=EyeAdaptationInverse}
    Epic {from=MaterialExpressionEyeAdaptationInverse_0, output=EyeAdaptationInverse}
   ]
   properties {}
  }
  MaterialExpressionReflectionVectorWS_1 {
   class=MaterialExpressionReflectionVectorWS
   inputs [CustomWorldNormal {from=MaterialExpressionNamedRerouteUsage_25, output=""}]
   properties {}
  }
  MaterialExpressionReflectionVectorWS_2 {class=MaterialExpressionReflectionVectorWS, inputs [CustomWorldNormal {from=MaterialExpressionTransform_1, output=""}], properties {}}
  MaterialExpressionReroute_1 {class=MaterialExpressionReroute, inputs [None {from=MaterialExpressionConstant_1}]}
  MaterialExpressionReroute_15 {class=MaterialExpressionReroute, inputs [None {from=MaterialExpressionFunctionInput_11}]}
  MaterialExpressionReroute_2 {class=MaterialExpressionReroute, inputs [None {from=MaterialExpressionConstant_5}]}
  MaterialExpressionReroute_7 {class=MaterialExpressionReroute, inputs [None {from=MaterialExpressionConstant_0}]}
  MaterialExpressionSaturate_13 {class=MaterialExpressionSaturate, inputs [None {from=MaterialExpressionMultiply_27}]}
  MaterialExpressionSaturate_15 {class=MaterialExpressionSaturate, inputs [None {from=MaterialExpressionMultiply_12}]}
  MaterialExpressionSaturate_16 {class=MaterialExpressionSaturate, inputs [None {from=MaterialExpressionDivide_1}]}
  MaterialExpressionScalarParameter_0 {class=MaterialExpressionScalarParameter, properties {defaultValue=1, group="-       EchoFrost Snow", parameterName=* Snow - Scale}}
  MaterialExpressionScalarParameter_1 {
   class=MaterialExpressionScalarParameter
   properties {defaultValue=1, group="-       EchoFrost Snow", parameterName=* Snow - Normal Blend Strength}
  }
  MaterialExpressionScalarParameter_13 {class=MaterialExpressionScalarParameter, properties {defaultValue=0.65, group=PostRVTSnow, parameterName=Epic_SnowSparkleEmissiveBrightness}}
  MaterialExpressionScalarParameter_17 {class=MaterialExpressionScalarParameter, properties {defaultValue=400, group=PostRVTSnow, parameterName=Epic_SnowSparklePingPongDistance}}
  MaterialExpressionScalarParameter_25 {class=MaterialExpressionScalarParameter, properties {defaultValue=6, parameterName=Epic_MeshSnowSparkleSize}}
  MaterialExpressionScalarParameter_27 {class=MaterialExpressionScalarParameter, properties {defaultValue=2000, group=PostRVTSnow, parameterName=Epic_SnowSparkleFadeDistance}}
  MaterialExpressionScalarParameter_31 {class=MaterialExpressionScalarParameter, properties {defaultValue=0.5, group="-       EchoFrost Snow", parameterName=* Snow - Specular}}
  MaterialExpressionScalarParameter_33 {class=MaterialExpressionScalarParameter, properties {defaultValue=0.8, group="-       EchoFrost Snow", parameterName=* Snow - Roughness}}
  MaterialExpressionScalarParameter_9 {class=MaterialExpressionScalarParameter, properties {defaultValue=2, group=PostRVTSnow, parameterName=Epic_SnowSparkleMaskExponent}}
  MaterialExpressionSetMaterialAttributes_0 {class=MaterialExpressionSetMaterialAttributes, inputs [MaterialAttributes {from="", output=""}]}
  MaterialExpressionSetMaterialAttributes_3 {
   class=MaterialExpressionSetMaterialAttributes
   inputs [
    MaterialAttributes {from="", output=""}
    BaseColor {from=MaterialExpressionShadingPathSwitch_2, output=""}
    Specular {from=MaterialExpressionScalarParameter_31, output=""}
    Roughness {from=MaterialExpressionScalarParameter_33, output=""}
    Opacity {from=MaterialExpressionConstant_11, output=""}
    Normal {from=MaterialExpressionShadingPathSwitch_0, output=""}
    EmissiveColor {from=MaterialExpressionShadingPathSwitch_1, output=""}
    OpacityMask {from=MaterialExpressionReroute_2, output=""}
    SubsurfaceColor {from=MaterialExpressionConstant_11, output=""}
   ]
   properties {attributeSetTypes [
     #2
     #3
     #4
     #5
     #1
     #6
     #7
     #8
    ]}
  }
  MaterialExpressionShadingPathSwitch_0 {
   class=MaterialExpressionShadingPathSwitch
   inputs [
    Default {from=MaterialExpressionLinearInterpolate_0, output=""}
    Deferred {from="", output=""}
    Forward {from="", output=""}
    Mobile {from=MaterialExpressionGetMaterialAttributes_0, output=Normal}
   ]
   properties {}
  }
  MaterialExpressionShadingPathSwitch_1 {
   class=MaterialExpressionShadingPathSwitch
   inputs [
    Default {from=MaterialExpressionQualitySwitch_2, output=""}
    Deferred {from="", output=""}
    Forward {from="", output=""}
    Mobile {from=MaterialExpressionConstant_11, output=""}
   ]
   properties {}
  }
  MaterialExpressionShadingPathSwitch_2 {
   class=MaterialExpressionShadingPathSwitch
   inputs [
    Default {from=MaterialExpressionMaterialFunctionCall_1, output=Result}
    Deferred {from="", output=""}
    Forward {from="", output=""}
    Mobile {from=MaterialExpressionVectorParameter_25, output=RGB}
   ]
   properties {}
  }
  MaterialExpressionSkyLightEnvMapSample_0 {
   class=MaterialExpressionSkyLightEnvMapSample
   inputs [
    Direction {from=MaterialExpressionReflectionVectorWS_1, output=""}
    Roughness {from="", output=""}
   ]
   properties {}
  }
  MaterialExpressionTextureCoordinate_0 {class=MaterialExpressionTextureCoordinate}
  MaterialExpressionTextureSampleParameter2D_0 {
   class=MaterialExpressionTextureSampleParameter2D
   inputs [
    UVs {from=MaterialExpressionFunctionInput_1, output=""}
    "DDX(UVs)" {from=MaterialExpressionFunctionInput_2, output=""}
    "DDY(UVs)" {from=MaterialExpressionFunctionInput_3, output=""}
    Apply View MipBias {from="", output=""}
   ]
   properties {
    group="-       EchoFrost Snow"
    mipValueMode=TMVM_Derivative
    parameterName=* Snow - Normal
    samplerSource=SSM_Wrap_WorldGroupSettings
    samplerType=SAMPLERTYPE_Normal
    texture=/Hera/Sets/EchoFrost/Textures/T_EchoFrost_SnowWall_A_N
   }
  }
  MaterialExpressionTextureSample_2 {
   class=MaterialExpressionTextureSample
   inputs [
    UVs {from=MaterialExpressionMultiply_16}
    Tex {from=""}
    Apply View MipBias {from=""}
   ]
   properties {
    samplerSource=SSM_Wrap_WorldGroupSettings
    samplerType=SAMPLERTYPE_Normal
    texture=/Game/Environments/Asteria/Terrain/Textures/T_SnowSparkleCircles_N
   }
  }
  MaterialExpressionTextureSample_4 {
   class=MaterialExpressionTextureSample
   inputs [
    UVs {from=MaterialExpressionMultiply_16}
    Tex {from=""}
    Apply View MipBias {from=""}
   ]
   properties {
    samplerSource=SSM_Wrap_WorldGroupSettings
    samplerType=SAMPLERTYPE_LinearColor
    texture=/Game/Environments/Asteria/Terrain/Textures/T_SnowSparkleCircles_M
   }
  }
  MaterialExpressionTextureSample_5 {
   class=MaterialExpressionTextureSample
   inputs [
    UVs {from=MaterialExpressionMultiply_21}
    Tex {from=""}
    Apply View MipBias {from=""}
   ]
   properties {
    samplerSource=SSM_Wrap_WorldGroupSettings
    samplerType=SAMPLERTYPE_Normal
    texture=&T_SnowSparkleCircles_N
   }
  }
  MaterialExpressionTextureSample_6 {
   class=MaterialExpressionTextureSample
   inputs [
    UVs {from=MaterialExpressionMultiply_21}
    Tex {from=""}
    Apply View MipBias {from=""}
   ]
   properties {
    samplerSource=SSM_Wrap_WorldGroupSettings
    samplerType=SAMPLERTYPE_LinearColor
    texture=&T_SnowSparkleCircles_M
   }
  }
  MaterialExpressionTransform_1 {class=MaterialExpressionTransform, inputs [None {from=MaterialExpressionNormalize_2}]}
  MaterialExpressionTransform_5 {
   class=MaterialExpressionTransform
   inputs [None {from=MaterialExpressionMaterialFunctionCall_17, output=World Normals}]
   properties {transformSourceType=TRANSFORMSOURCE_World, transformType=TRANSFORM_Tangent}
  }
  MaterialExpressionVectorParameter_25 {
   class=MaterialExpressionVectorParameter
   properties {group="-       EchoFrost Snow", parameterName=* Snow - Color, defaultValue=<rgba 0.3763 0.3916 0.4072 1>}
  }
  MaterialExpressionVectorParameter_3 {
   class=MaterialExpressionVectorParameter
   properties {group="-       EchoFrost Snow", parameterName="* Snow - Color (Night)", defaultValue=<rgba 0.2574 0.3138 0.3715 1>}
  }
 }}
