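# Material function that adds EchoFrost ice, with an optional snow layer, over an architecture material's attributes, switchable by a static input.
# Ice uses dithered triplanar surface textures, parallax internal layers and emissive tinted by world-space height gradients, which can be overridden inside box regions around castle, towers and bridge.
# Height-lerped noise masks ice; normal-up, gradient, distance-field AO and breakup mask snow from MF_EchoFrost_Snow.
# Quality and mobile switches trim costly features.
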MaterialFunction /Hera/Sets/EchoFrost/Materials/MaterialFunctions/MF_EchoFrost_Architecture_Ice
graph {nodes [
  DotProduct()
  FunctionInput() {inputName="\\ --- Masks --- /", inputType=FunctionInput_Scalar, sortPriority=20}
  FunctionInput() {inputName="\\ --- Dependencies --- /", inputType=FunctionInput_Scalar, sortPriority=800}
  n1 = NamedRerouteUsage()
  n2 = NamedRerouteUsage()
  n3 = NamedRerouteUsage()
  n4 = NamedRerouteUsage()
  n5 = NamedRerouteUsage()
  n6 = NamedRerouteUsage()
  n7 = ComponentMask(None=Transform(None=GetMaterialAttributes(None=@n6) {attributeGetTypes [#1]}.Normal)) {b=true}
  n8 = Multiply(
   A=Add(
    A=Saturate(None=@n7)
    B=ScalarParameter() {defaultValue=-0.1, group="-       EchoFrost Snow Masking", parameterName=* Snow - Z UP Normal - Bias, sortPriority=400}
   )
   B=ScalarParameter() {defaultValue=5, group="-       EchoFrost Snow Masking", parameterName=* Snow - Z UP Normal - Intensity, sortPriority=401}
  )
  n9 = Add(
   A=ScalarParameter() {defaultValue=-0.33, group="-       EchoFrost Snow Masking", parameterName=* Snow - Gradient - Bias, sortPriority=51}
   B=OneMinus(None=NamedRerouteUsage())
  )
  n10 = DistanceFieldApproxAO!(
   World Position
   Normal
   BaseDistance=ScalarParameter() {defaultValue=10, group="-       EchoFrost Snow Masking", parameterName=* Snow - DFAO - Distance, sortPriority=51}
   Radius=ScalarParameter() {defaultValue=300, group="-       EchoFrost Snow Masking", parameterName=* Snow - DFAO - Radius, sortPriority=51}
  ) {numSteps=3}
  n11 = Divide(A=Subtract(A=PixelDepth().R, B=Constant() {r=5000}), B=Constant() {r=10000})
  n12 = Saturate(None=Subtract(A=OneMinus(None=@n10), B=Saturate(None=@n11)))
  n13 = ScalarParameter() {defaultValue=1, group="-       EchoFrost Snow Masking", parameterName=* Snow - DFAO - Power, sortPriority=51}
  n14 = Constant()
  n15 = QualitySwitch(
   Default=@n14
   Epic=Multiply(
    A=Power(Base=@n12, Exp=@n13)
    B=ScalarParameter() {defaultValue=1, group="-       EchoFrost Snow Masking", parameterName="* Snow - DFAO - Mult (hardness)", sortPriority=51}
   )
  )
  n16 = OneMinus(None=ScalarParameter() {defaultValue=1, group="-       EchoFrost Snow Masking", parameterName=* Snow - DFAO - Gradient Intensity, sortPriority=51})
  n17 = Add(A=Multiply(A=ShadingPathSwitch(Default=@n15, Mobile=@n14), B=Max(A=@n9, B=@n16)), B=@n9)
  n18 = ScalarParameter() {defaultValue=1, group="-       EchoFrost Snow Masking", parameterName=* Snow - DFAO - Gradient Blend}
  n19 = VectorParameter() {
   group="-       EchoFrost Snow Masking"
   parameterName=* Snow - WS Noise - Channel Weights
   sortPriority=460
   channelNames {a="", b=Blue, g=Green, r=Red}
   defaultValue=<rgba 0 1 0 0>
  }
  n20 = DotProduct(A=@n19.RGB, B=NamedRerouteUsage())
  n21 = ScalarParameter() {defaultValue=0.3, group="-       EchoFrost Snow Masking", parameterName=* Snow - WS Noise - Bias, sortPriority=475}
  n22 = Saturate(None=Multiply(
    A=Add(A=@n20, B=@n21)
    B=ScalarParameter() {defaultValue=1, group="-       EchoFrost Snow Masking", parameterName=* Snow - WS Noise - Intensity, sortPriority=476}
   ))
  n23 = Add(
   A=Add(A=Saturate(None=@n8), B=LinearInterpolate(A=@n9, B=@n17, Alpha=@n18))
   B=Multiply(
    A=ScalarParameter() {defaultValue=0.5, group="-       EchoFrost Snow Masking", parameterName=* Snow - WS Noise - Multiply, sortPriority=477}
    B=@n22
   )
  )
  n24 = Saturate(None=@n23)
  n25 = NamedRerouteUsage()
  n26 = Saturate(None=LinearInterpolate(
    A=ScalarParameter() {defaultValue=0.3, group="-       EchoFrost Snow Masking", parameterName=* Snow - Breakup - Min, sortPriority=550}
    B=ScalarParameter() {defaultValue=0.7, group="-       EchoFrost Snow Masking", parameterName=* Snow - Breakup - Max, sortPriority=551}
    Alpha=GetMaterialAttributes(None=@n25) {attributeGetTypes [#2, #3]}.Roughness
   ))
  n27 = Saturate(None=MaterialFunctionCall!(
    1=ScalarParameter() {defaultValue=1, group="-       EchoFrost Snow Masking", parameterName=* Snow - Breakup Contrast - Layer 1, sortPriority=549}
    2=ScalarParameter() {defaultValue=1, group="-       EchoFrost Snow Masking", parameterName=* Snow - Breakup Contrast - Layer 2, sortPriority=549}
    3=ScalarParameter() {defaultValue=1, group="-       EchoFrost Snow Masking", parameterName=* Snow - Breakup Contrast - Layer 3, sortPriority=549}
    4=ScalarParameter() {defaultValue=1, group="-       EchoFrost Snow Masking", parameterName=* Snow - Breakup Contrast - Layer 4, sortPriority=549}
   ) {materialFunction=/Game/Global/Base4Layers_2022/MaterialFunctions/MF_UV2LerpScalar_Layers_New}.Result)
  n28 = Saturate(None=Multiply(
    A=ScalarParameter() {defaultValue=1, group="-       EchoFrost Snow Masking", parameterName=* Snow - Breakup - Intensity, sortPriority=552}
    B=LinearInterpolate(B=@n26, Alpha=@n27) {constA=0.5}
   ))
  n29 = Saturate(None=Add(A=Multiply(A=@n24) {constB=2, desc=2}, B=Subtract(A=@n28) {desc=1}))
  n30 = LinearInterpolate(
   A=@n29
   B=Multiply(A=@n28, B=@n29)
   Alpha=ScalarParameter() {defaultValue=1, group="-       EchoFrost Snow Masking", parameterName=* Snow - Breakup - Multiply, sortPriority=552}
  )
  n31 = Add(A=@n30, B=ScalarParameter() {group="-       EchoFrost Snow Masking", parameterName=* Snow - Masking - Bias, sortPriority=18})
  n32 = MaterialFunctionCall(
   In=Saturate(None=@n31)
   Contrast=ScalarParameter() {defaultValue=2, group="-       EchoFrost Snow Masking", parameterName=* Snow - Masking - Contrast, sortPriority=18}
  ) {materialFunction=&CheapContrast}
  n33 = ScalarParameter() {defaultValue=1, group="-       EchoFrost Snow Masking", parameterName=* Snow - Masking - Layer 1, sortPriority=900}
  n34 = ScalarParameter() {defaultValue=1, group="-       EchoFrost Snow Masking", parameterName=* Snow - Masking - Layer 2, sortPriority=900}
  n35 = ScalarParameter() {defaultValue=1, group="-       EchoFrost Snow Masking", parameterName=* Snow - Masking - Layer 3, sortPriority=900}
  n36 = ScalarParameter() {defaultValue=1, group="-       EchoFrost Snow Masking", parameterName=* Snow - Masking - Layer 4, sortPriority=900}
  n37 = Multiply(
   A=@n32.Result
   B=Saturate(None=MaterialFunctionCall(1=@n33, 2=@n34, 3=@n35, 4=@n36) {materialFunction=&MF_UV2LerpScalar_Layers_New}.Result)
  )
  n38 = StaticBoolParameter() {group="-       EchoFrost Snow Masking", parameterName=* Snow - Masking - Use Layer Masking Values, sortPriority=-50}
  n39 = NamedRerouteUsage()
  n40 = ScalarParameter() {group="-       EchoFrost Ice Masking", parameterName=* Ice - Breakup - Min, sortPriority=100}
  n41 = ScalarParameter() {defaultValue=1, group="-       EchoFrost Ice Masking", parameterName=* Ice - Breakup - Max, sortPriority=100}
  n42 = GetMaterialAttributes(None=NamedRerouteUsage()) {attributeGetTypes [#2]}
  n43 = Multiply(
   A=ScalarParameter() {defaultValue=0.25, group="-       EchoFrost Ice Masking", parameterName=* Ice - Breakup - Intensity, sortPriority=100}
   B=Saturate(None=LinearInterpolate(A=@n40, B=@n41, Alpha=@n42.Roughness))
  )
  n44 = VectorParameter() {
   group="-       EchoFrost Ice Masking"
   parameterName=* Ice - WS Noise - Channel Weights
   sortPriority=200
   channelNames {a="", b=Blue, g=Green, r=Red}
   defaultValue=<rgba 1 0 0 0>
  }
  n45 = DotProduct(A=@n44.RGB, B=NamedRerouteUsage())
  n46 = ScalarParameter() {group="-       EchoFrost Ice Masking", parameterName=* Ice - WS Noise - Bias, sortPriority=205}
  n47 = Saturate(None=Multiply(
    A=Add(A=@n45, B=@n46)
    B=ScalarParameter() {defaultValue=1, group="-       EchoFrost Ice Masking", parameterName=* Ice - WS Noise - Intensity, sortPriority=206}
   ))
  n48 = MaterialFunctionCall(Transition Phase=OneMinus(None=NamedRerouteUsage()), Height Texture=Saturate(None=Add(A=@n43, B=@n47))) {materialFunction=&HeightLerp}
  n49 = ScalarParameter() {defaultValue=0.2, group="-       EchoFrost Ice Masking", parameterName=* Ice - Masking - Bias, sortPriority=18}
  n50 = MaterialFunctionCall(
   In=Saturate(None=Add(A=@n48.Lerp Alpha No Contrast, B=@n49))
   Contrast=ScalarParameter() {defaultValue=8, group="-       EchoFrost Ice Masking", parameterName=* Ice - Masking - Contrast, sortPriority=18}
  ) {materialFunction=&CheapContrast}
  n51 = Saturate(None=MaterialFunctionCall(
    1=ScalarParameter() {defaultValue=1, group="-       EchoFrost Ice Masking", parameterName=* Ice - Masking - Layer 1, sortPriority=900}
    2=ScalarParameter() {defaultValue=1, group="-       EchoFrost Ice Masking", parameterName=* Ice - Masking - Layer 2, sortPriority=900}
    3=ScalarParameter() {defaultValue=1, group="-       EchoFrost Ice Masking", parameterName=* Ice - Masking - Layer 3, sortPriority=900}
    4=ScalarParameter() {defaultValue=1, group="-       EchoFrost Ice Masking", parameterName=* Ice - Masking - Layer 4, sortPriority=900}
   ) {materialFunction=&MF_UV2LerpScalar_Layers_New}.Result)
  n52 = BlendMaterialAttributes(
   A=@n39
   B=NamedRerouteUsage()
   Alpha=StaticSwitch(
    True=Multiply(A=@n50.Result, B=@n51)
    False=@n50.Result
    Value=StaticBoolParameter() {group="-       EchoFrost Ice Masking", parameterName=* Ice - Masking - Use Layer Masking Values, sortPriority=-50}
   )
  )
  n53 = ShadingPathSwitch(Default=QualitySwitch(Default=@n52, Low=@n39, High=@n52, Medium=@n52, Epic=@n52), Mobile=@n39)
  n54 = StaticSwitch(
   True=BlendMaterialAttributes(
    A=@n53
    B=MaterialFunctionCall!(
     Input MaterialAttributes=@n1
     "\\ --- Dependencies --- /"
     Baked Curvature
     Baked AO
     Painted Mask
     DFAO
     Z Gradient
     WS Noise Masks
     IceSnow WS - UVs=@n2
     IceSnow WS - DDX=@n3
     IceSnow WS - DDY=@n4
     IceSnow WS - Mask=@n5
    ) {materialFunction=&/MF_EchoFrost_Snow}.Snow MaterialAttributes
    Alpha=StaticSwitch(True=@n37, False=@n32.Result, Value=@n38)
   )
   False=@n53
   Value=StaticBoolParameter() {defaultValue=true, group="-       EchoFrost Ice", parameterName=Use EchoFrost Snow Layer, sortPriority=-9}
  )
  n55 = StaticSwitch(
   True=SetMaterialAttributes!(MaterialAttributes=@n54, BaseColor=Constant(), EmissiveColor=NamedRerouteUsage()) {attributeSetTypes [#4, #5]}
   False=@n54
   Value=StaticBoolParameter() {group="-       EchoFrost Gradients", parameterName=DEBUG - Gradients, sortPriority=-100}
  )
  FunctionOutput(None=StaticSwitch(True=@n55, False=NamedRerouteUsage(), Value=FunctionInput() {inputName=Use EchoFrost Ice Layer, inputType=FunctionInput_StaticBool})) {outputName=MaterialAttributes, sortPriority=-10}
  n56 = VectorParameter() {
   group="-       EchoFrost Ice Masking"
   parameterName=* Ice - WS Noise - XYZ Texture Scale
   sortPriority=453
   channelNames {a="", b=Z - Scale, g=Y - Scale, r=X - Scale}
   defaultValue=<rgba 1200 1200 2000 1>
  }
  n57 = Constant() {r=75}
  n58 = Constant() {r=30}
  n59 = MaterialFunctionCall!(
   Absolute World Position
   "Normal (WS)"
   Texture Scale=@n56.RGB
   Triplanar Fade=MaterialFunctionCall!(Low=@n57, High=@n58, Epic=@n58, --------=@n58, Mobile=@n57, Medium=@n58) {materialFunction=/FortniteBaseMaterials/Functions/MF_QualitySwitch_Scalar}.Result
   Derivatives Scale
  ) {materialFunction=/BRCoreMaterials/Materials/MaterialFunctions/MF_TriplanarDithered_WavyDistortion}
  n60 = TextureObjectParameter() {
   group="-       EchoFrost Ice Masking"
   parameterName=* Ice - WS Noise - Masks Texture
   samplerType=SAMPLERTYPE_Masks
   sortPriority=452
   texture=/Hera/Sets/EchoFrost/Textures/EchoFrost_IceSnow_Noise_MSK
  }
  n61 = TextureSample!(UVs=@n59.UVs, Tex=@n60, "DDX(UVs)"=@n59.UVs, "DDY(UVs)"=@n59.UVs, Apply View MipBias) {mipValueMode=TMVM_Derivative}
  n62 = Add(A=@n61.B, B=ScalarParameter() {group="-       EchoFrost Snow Masking", parameterName=* Snow - Small Splotches - Bias, sortPriority=500})
  n63 = ScalarParameter() {defaultValue=1, group="-       EchoFrost Snow Masking", parameterName=* Snow - Small Splotches - Intensity, sortPriority=500}
  Max(A=@n22, B=Saturate(None=Multiply(A=@n62, B=@n63)))
  Multiply()
  n64 = VectorParameter() {
   group="-       EchoFrost Ice"
   parameterName=* Internal Gradient - Bottom Color
   sortPriority=351
   defaultValue=<rgba 0.7758 0.1301 0.8796 1>
  }
  n65 = VectorParameter() {group="-       EchoFrost Ice", parameterName=* Internal Gradient - Top Color, sortPriority=350, defaultValue=<rgba 0.1475 0.7454 0.206 1>}
  n66 = NamedRerouteUsage()
  n67 = ComponentMask(None=ObjectPositionWS()) {b=true}
  n68 = ScalarParameter() {
   defaultValue=7613.4438
   group="-       EchoFrost Gradients"
   parameterName=* Gradient - Global - Bottom
   primitiveDataIndex=3
   sortPriority=620
  }
  n69 = ScalarParameter() {
   defaultValue=11553.0449
   group="-       EchoFrost Gradients"
   parameterName=* Gradient - Global - Top
   primitiveDataIndex=3
   sortPriority=621
  }
  n70 = ScalarParameter() {bUseCustomPrimitiveData=true, group="-       EchoFrost Ice", parameterName=* Gradient - Global - CPD Offset, primitiveDataIndex=2}
  n71 = Add(A=@n70, B=@n68)
  n72 = Add(A=@n70, B=@n69)
  n73 = NamedRerouteUsage()
  n74 = MaterialFunctionCall!(min=@n68, max=@n69, value=NamedRerouteUsage()) {materialFunction=/Game/Effects/Fort_Effects/Materials/_Functions/InverseLerp_MF}
  n75 = StaticBoolParameter() {defaultValue=true, group="-       EchoFrost Gradients", parameterName=* Gradient - Use WS Box Gradients, sortPriority=-50}
  n76 = ScalarParameter() {defaultValue=8124.3379, group="-       EchoFrost Gradients", parameterName=* Gradient - A - Bottom, primitiveDataIndex=3, sortPriority=630}
  n77 = ScalarParameter() {defaultValue=15852.7266, group="-       EchoFrost Gradients", parameterName=* Gradient - A - Top, primitiveDataIndex=3, sortPriority=631}
  n78 = NamedRerouteUsage()
  n79 = VectorParameter() {
   group="-       EchoFrost Gradients"
   parameterName="* Gradient - A - Center (Castle)"
   sortPriority=634
   defaultValue=<rgba 5036.5684 -51413.9883 7296 1>
  }
  n80 = ComponentMask(None=@n79.RGB) {g=true, r=true}
  n81 = VectorParameter() {group="-       EchoFrost Gradients", parameterName=* Gradient - A - Bounds, sortPriority=634, defaultValue=<rgba 15067.0713 6758.9209 0 1>}
  n82 = ComponentMask(None=@n81.RGB) {g=true, r=true}
  n83 = ScalarParameter() {defaultValue=1500, group="-       EchoFrost Gradients", parameterName=* Gradient - A - Edge Falloff, sortPriority=634}
  n84 = ScalarParameter() {defaultValue=10321.0186, group="-       EchoFrost Gradients", parameterName=* Gradient - B - Bottom, primitiveDataIndex=3, sortPriority=640}
  n85 = ScalarParameter() {defaultValue=14032.0137, group="-       EchoFrost Gradients", parameterName=* Gradient - B - Top, primitiveDataIndex=3, sortPriority=641}
  n86 = VectorParameter() {
   group="-       EchoFrost Gradients"
   parameterName="* Gradient - B - Center (Castle Tower)"
   sortPriority=644
   defaultValue=<rgba 5629.6172 -54579.4062 10023.0498 1>
  }
  n87 = VectorParameter() {group="-       EchoFrost Gradients", parameterName=* Gradient - B - Bounds, sortPriority=644, defaultValue=<rgba 4000 2000 0 1>}
  n88 = LinearInterpolate(
   A=LinearInterpolate(
    A=@n74.Clamped
    B=MaterialFunctionCall(min=@n76, max=@n77, value=@n78) {materialFunction=&InverseLerp_MF}.Clamped
    Alpha=MaterialFunctionCall!(A, B=@n80, Bounds=@n82, Edge Falloff=@n83) {materialFunction=/Engine/Functions/Engine_MaterialFunctions01/Gradient/BoxMask-2D}.Result
   )
   B=MaterialFunctionCall(min=@n84, max=@n85, value=NamedRerouteUsage()) {materialFunction=&InverseLerp_MF}.Clamped
   Alpha=MaterialFunctionCall(
    B=ComponentMask(None=@n86.RGB) {g=true, r=true}
    Bounds=ComponentMask(None=@n87.RGB) {g=true, r=true}
    Edge Falloff=ScalarParameter() {defaultValue=1200, group="-       EchoFrost Gradients", parameterName=* Gradient - B - Edge Falloff, sortPriority=644}
   ) {materialFunction=&BoxMask-2D}.Result
  )
  n89 = ScalarParameter() {defaultValue=5828.6055, group="-       EchoFrost Gradients", parameterName=* Gradient - C - Bottom, primitiveDataIndex=3, sortPriority=650}
  n90 = ScalarParameter() {defaultValue=10996.6787, group="-       EchoFrost Gradients", parameterName=* Gradient - C - Top, primitiveDataIndex=3, sortPriority=651}
  n91 = MaterialFunctionCall(min=@n89, max=@n90, value=NamedRerouteUsage()) {materialFunction=&InverseLerp_MF}
  n92 = VectorParameter() {
   group="-       EchoFrost Gradients"
   parameterName="* Gradient - C - Center (Bridge)"
   sortPriority=654
   defaultValue=<rgba 5721.0142 -39364.2031 7296 1>
  }
  n93 = VectorParameter() {group="-       EchoFrost Gradients", parameterName=* Gradient - C - Bounds, sortPriority=654, defaultValue=<rgba 2000 4500 0 1>}
  n94 = MaterialFunctionCall(
   B=ComponentMask(None=@n92.RGB) {g=true, r=true}
   Bounds=ComponentMask(None=@n93.RGB) {g=true, r=true}
   Edge Falloff=ScalarParameter() {defaultValue=800, group="-       EchoFrost Gradients", parameterName=* Gradient - C - Edge Falloff, sortPriority=654}
  ) {materialFunction=&BoxMask-2D}
  n95 = ScalarParameter() {defaultValue=6144, group="-       EchoFrost Gradients", parameterName=* Gradient - D - Bottom, primitiveDataIndex=3, sortPriority=660}
  n96 = ScalarParameter() {defaultValue=7000, group="-       EchoFrost Gradients", parameterName=* Gradient - D - Top, primitiveDataIndex=3, sortPriority=661}
  n97 = NamedRerouteUsage()
  n98 = VectorParameter() {
   group="-       EchoFrost Gradients"
   parameterName="* Gradient - D - Center (Valley Tower)"
   sortPriority=664
   defaultValue=<rgba 1024 -36352 6144 1>
  }
  n99 = ComponentMask(None=@n98.RGB) {g=true, r=true}
  n100 = VectorParameter() {group="-       EchoFrost Gradients", parameterName=* Gradient - D - Bounds, sortPriority=664, defaultValue=<rgba 1500 1500 0 1>}
  n101 = ComponentMask(None=@n100.RGB) {g=true, r=true}
  n102 = ScalarParameter() {defaultValue=400, group="-       EchoFrost Gradients", parameterName=* Gradient - D - Edge Falloff, sortPriority=664}
  n103 = ScalarParameter() {defaultValue=7100, group="-       EchoFrost Gradients", parameterName=* Gradient - E - Bottom, primitiveDataIndex=3, sortPriority=670}
  n104 = ScalarParameter() {defaultValue=8722.5928, group="-       EchoFrost Gradients", parameterName=* Gradient - E - Top, primitiveDataIndex=3, sortPriority=671}
  n105 = VectorParameter() {
   group="-       EchoFrost Gradients"
   parameterName="* Gradient - E - Center (Castle Extension)"
   sortPriority=674
   defaultValue=<rgba 5630.3623 -29579.8926 7296 1>
  }
  n106 = VectorParameter() {group="-       EchoFrost Gradients", parameterName=* Gradient - E - Bounds, sortPriority=674, defaultValue=<rgba 5356.0669 11708.7246 0 1>}
  n107 = LinearInterpolate(
   A=LinearInterpolate(
    A=LinearInterpolate(A=@n88, B=@n91.Clamped, Alpha=@n94.Result)
    B=MaterialFunctionCall(min=@n95, max=@n96, value=@n97) {materialFunction=&InverseLerp_MF}.Clamped
    Alpha=MaterialFunctionCall(B=@n99, Bounds=@n101, Edge Falloff=@n102) {materialFunction=&BoxMask-2D}.Result
   )
   B=MaterialFunctionCall(min=@n103, max=@n104, value=NamedRerouteUsage()) {materialFunction=&InverseLerp_MF}.Clamped
   Alpha=MaterialFunctionCall(
    B=ComponentMask(None=@n105.RGB) {g=true, r=true}
    Bounds=ComponentMask(None=@n106.RGB) {g=true, r=true}
    Edge Falloff=ScalarParameter() {defaultValue=800, group="-       EchoFrost Gradients", parameterName=* Gradient - E - Edge Falloff, sortPriority=674}
   ) {materialFunction=&BoxMask-2D}.Result
  )
  n108 = StaticBoolParameter() {
   desc=When enabled gradient will stay consistent with the mesh when moved
   group="-       EchoFrost Gradients"
   parameterName=* Gradient - Use LocalWS Gradient
   sortPriority=-50
  }
  n109 = StaticSwitch(
   True=MaterialFunctionCall(min=Add(A=@n67, B=@n68), max=Add(A=@n67, B=@n69), value=NamedRerouteUsage()) {materialFunction=&InverseLerp_MF}.Clamped
   False=StaticSwitch(
    True=MaterialFunctionCall(min=@n71, max=@n72, value=@n73) {materialFunction=&InverseLerp_MF}.Clamped
    False=StaticSwitch(True=@n107, False=@n74.Clamped, Value=@n75)
    Value=StaticBoolParameter() {group="-       EchoFrost Gradients", parameterName=* Gradient - Use CPD Gradient, sortPriority=-50}
   )
   Value=@n108
  )
  n110 = StaticBoolParameter() {
   desc="Masks Ice from Layer2 by default, can change"
   group="-       EchoFrost Gradients"
   parameterName=* Gradient - Use 3D Box Gradient
   sortPriority=-50
  }
  n111 = OneMinus(None=Saturate(None=StaticSwitch(True=@n66, False=@n109, Value=@n110)))
  n112 = ScalarParameter() {defaultValue=4, group="-       EchoFrost Gradients", parameterName=* Gradient - Power, sortPriority=600}
  n113 = Saturate(None=OneMinus(None=Power(Base=@n111, Exp=@n112)))
  NamedRerouteDeclaration(None=Multiply(
    A=LinearInterpolate(A=@n64.RGB, B=@n65.RGB, Alpha=@n113)
    B=ScalarParameter() {defaultValue=8, desc=8, group="-       EchoFrost Ice", parameterName=* Internal Gradient - Strength, sortPriority=600}
   ) {constB=5}) {name=Ice - Gradients, nodeColor=<rgba 0 0.9766 1 1>}
  n114 = VertexNormalWS()
  n115 = CameraVectorWS()
  n116 = OneMinus(None=Saturate(None=DotProduct(A=@n114, B=@n115)))
  NamedRerouteDeclaration(None=OneMinus(None=NamedRerouteDeclaration(None=@n116) {name=Fresnel, nodeColor=<rgba 0.9922 1 0 1>})) {name=Fresnel_Inverted, nodeColor=<rgba 0.9922 1 0 1>}
  n117 = MaterialFunctionCall(
   Absolute World Position=WorldPosition().XYZ
   Texture Scale=ScalarParameter() {defaultValue=700, group="-       EchoFrost Ice", parameterName=* Ice - Surface - Scale, sortPriority=60}
  ) {materialFunction=&MF_TriplanarDithered_WavyDistortion}
  NamedRerouteDeclaration(None=@n117.DDX) {name=Ice/Snow - DDX, nodeColor=<rgba 0.9219 0 1 1>}
  NamedRerouteDeclaration(None=@n117.DDY) {name=Ice/Snow - DDY, nodeColor=<rgba 0.9219 0 1 1>}
  NamedRerouteDeclaration(None=@n117.Mask) {name=Ice/Snow - Mask, nodeColor=<rgba 0.9219 0 1 1>}
  NamedRerouteDeclaration(None=FunctionInput() {inputName=Z Gradient, inputType=FunctionInput_Scalar, sortPriority=814}) {name=_Z Gradient, nodeColor=<rgba 1 0 0.0525 1>}
  NamedRerouteDeclaration(None=FunctionInput(Preview=SetMaterialAttributes!(MaterialAttributes)) {inputName=Input MaterialAttributes, inputType=FunctionInput_MaterialAttributes}) {name=_Input MaterialAttributes, nodeColor=<rgba 0.1016 0 1 1>}
  n118 = MaterialFunctionCall() {materialFunction=&ObjectLocalBounds}
  n119 = Subtract(A=LocalPosition().XYZ, B=@n118.Local Bounds Minimum)
  NamedRerouteDeclaration(None=ComponentMask(None=Divide(A=@n119, B=@n118.Local Bounds Size)) {b=true}) {name=_Z Gradient, nodeColor=<rgba 0 0 0.5469 1>}
  n120 = ScalarParameter() {
   defaultValue=1
   desc=Happens Post Eye Adaption Inverse adjustments
   group="-       EchoFrost Ice"
   parameterName="* Internal Light - Global Emissive (Lumen)"
   sortPriority=100
  }
  n121 = GetMaterialAttributes(None=NamedRerouteUsage()) {attributeGetTypes [
    #4
    #6
    #3
    #2
    #7
    #5
   ]}
  n122 = Multiply(
   A=Multiply(
    A=Multiply(A=@n121.BaseColor, B=@n121.BaseColor)
    B=ScalarParameter() {defaultValue=2, group="-       EchoFrost Ice", parameterName=* Emissive - Ice BaseColor Contribution, sortPriority=100}
   )
   B=ScalarParameter() {defaultValue=0.3, group="-       EchoFrost Ice", parameterName=* Emissive - Global Intensity, sortPriority=100}
  )
  n123 = ScalarParameter() {defaultValue=0.1, group="-       EchoFrost Ice", parameterName=* Internal Layers - Parallax - Emissive Intensity, sortPriority=307}
  n124 = Multiply(A=NamedRerouteUsage(), B=@n123) {constB=0.1}
  n125 = ShadingPathSwitch(Default=QualitySwitch(Default=Add(A=@n122, B=@n124), Low=@n122, Medium=@n122), Mobile=@n122)
  n126 = VectorParameter() {
   group="-       EchoFrost Ice"
   parameterName=* Ice - Surface - Roughness - Scalability
   sortPriority=30
   channelNames {a=Unused, b=Unused, g=Medium, r=Low}
   defaultValue=<rgba 1 1 1 1>
  }
  n127 = ScalarParameter() {defaultValue=2, group="-       EchoFrost Ice", parameterName=* Internal Layers - Parallax - BaseColor Intensity, sortPriority=306}
  n128 = NamedRerouteUsage()
  n129 = Add(A=Multiply(A=@n127, B=Multiply(A=@n128, B=@n121.BaseColor)), B=@n121.BaseColor)
  n130 = SetMaterialAttributes!(
   MaterialAttributes=NamedRerouteUsage()
   BaseColor=ShadingPathSwitch(Default=QualitySwitch(Default=@n129, Low=@n121.BaseColor, High=@n129, Medium=@n121.BaseColor, Epic=@n129), Mobile=@n121.BaseColor)
   Specular=@n121.Specular
   Roughness=QualitySwitch(
    Default=@n121.Roughness
    Low=Multiply(A=@n121.Roughness, B=@n126.Low)
    High=@n121.Roughness
    Medium=Multiply(A=@n121.Roughness, B=@n126.Medium)
    Epic=@n121.Roughness
   )
   EmissiveColor=RayTracingQualitySwitch(Normal=EyeAdaptationInverse(LightValueInput=@n125).EyeAdaptationInverse, RayTraced=Multiply(A=@n125, B=@n120))
   Normal=GetMaterialAttributes(None=NamedRerouteUsage()) {attributeGetTypes [#1]}.Normal
  ) {attributeSetTypes [
    #4
    #3
    #2
    #5
    #1
   ]}
  NamedRerouteDeclaration(None=@n130) {name=Ice Elements Combined, nodeColor=<rgba 0.9922 1 0 1>}
  n131 = VectorParameter() {group="-       EchoFrost Ice", parameterName=* Ice - BaseColor - Tint, sortPriority=50, defaultValue=<rgba 0.0488 0.1064 0.151 1>}
  n132 = Add(A=@n131.RGB, B=@n131.RGB)
  n133 = ScalarParameter() {defaultValue=2, group="-       EchoFrost Ice", parameterName=* Ice - BaseColor - Fresnel Intensity, sortPriority=50}
  n134 = VectorParameter() {
   group="-       EchoFrost Ice"
   parameterName=* Ice - BaseColor - Z up Normal Tint
   sortPriority=50
   defaultValue=<rgba 0.2043 0.3302 0.4087 1>
  }
  n135 = LinearInterpolate(A=LinearInterpolate(A=@n131.RGB, B=Multiply(A=@n132, B=@n133), Alpha=NamedRerouteUsage()), B=@n134.RGB, Alpha=NamedRerouteUsage())
  n136 = ScalarParameter() {defaultValue=0.5, group="-       EchoFrost Ice", parameterName=* Ice - Surface - Specular, sortPriority=40}
  n137 = ScalarParameter() {defaultValue=1, group="-       EchoFrost Ice", parameterName=* Ice - Surface - Specular - Surface Noise - Max, sortPriority=43}
  n138 = TextureObjectParameter() {
   group="-       EchoFrost Ice"
   parameterName=* Ice - Surface - SpecularMasks
   samplerType=SAMPLERTYPE_Masks
   sortPriority=66
   texture=/Hera/Sets/EchoFrost/Textures/T_EchoFrost_IceSurface_Facets_A_S
  }
  MaterialExpressionTextureSample_7 = TextureSample(UVs=@n117.UVs, Tex=@n138, "DDX(UVs)"=@n117.UVs, "DDY(UVs)"=@n117.UVs) {
   mipValueMode=TMVM_Derivative
   samplerSource=SSM_Wrap_WorldGroupSettings
   samplerType=SAMPLERTYPE_Normal
   texture=/Game/Environments/Asteria/Terrain/Textures/T_Asteria_RockBase_01_D
  }
  n139 = Convert(None=@MaterialExpressionTextureSample_7.RGB) {
   nodeName=Break Float3
   convertInputs [{
     type=Vector3
     expressionInput {
      expression=&self:MaterialExpressionTextureSample_7
      inputName=None
      mask=1
      maskA=0
      maskB=1
      maskG=1
      maskR=1
      outputIndex=0
     }
    }]
   convertOutputs [{type=Scalar}, {type=Scalar}, {type=Scalar}]
  }
  n140 = Multiply(
   A=LinearInterpolate(
    A=ScalarParameter() {group="-       EchoFrost Ice", parameterName=* Ice - Surface - Specular - Surface Noise - Min, sortPriority=42}
    B=@n137
    Alpha=@n139
   )
   B=@n136
  )
  n141 = ScalarParameter() {defaultValue=0.5, group="-       EchoFrost Ice", parameterName=* Ice - Surface - Specular - Surface Noise Amount, sortPriority=41}
  n142 = ScalarParameter() {defaultValue=0.4, group="-       EchoFrost Ice", parameterName=* Ice - Surface - Roughness, sortPriority=30}
  n143 = ScalarParameter() {defaultValue=2, group="-       EchoFrost Ice", parameterName=* Ice - Surface - Roughness - Surface Noise - Max, sortPriority=33}
  n144 = LinearInterpolate(A=ScalarParameter() {group="-       EchoFrost Ice", parameterName=* Ice - Surface - Roughness - Surface Noise - Min}, B=@n143, Alpha=@n139)
  n145 = ScalarParameter() {defaultValue=0.5, group="-       EchoFrost Ice", parameterName=* Ice - Surface - Roughness - Surface Noise Amount, sortPriority=31}
  n146 = NamedRerouteUsage()
  n147 = TextureObjectParameter() {
   group="-       EchoFrost Ice"
   parameterName="* Ice - Surface - Normal (Also affects Distortion)"
   samplerType=SAMPLERTYPE_Normal
   sortPriority=65
   texture=/Hera/Sets/EchoFrost/Textures/T_EchoFrost_IceSurface_Flat_A_N
  }
  n148 = TextureSample(UVs=@n117.UVs, Tex=@n147, "DDX(UVs)"=@n117.UVs, "DDY(UVs)"=@n117.UVs) {
   mipValueMode=TMVM_Derivative
   samplerSource=SSM_Wrap_WorldGroupSettings
   samplerType=SAMPLERTYPE_Normal
   texture=&T_Asteria_RockBase_01_D
  }
  n149 = Transform(None=MaterialFunctionCall!(Tangent Normals=@n148.RGB, Vertex Normals, Triplanar Mask=@n117.Mask) {materialFunction=/Game/Athena/Environments/MaterialFunctions/MF_TriplanarNormalTangentToWorld}.World Normals) {transformSourceType=TRANSFORMSOURCE_World, transformType=TRANSFORM_Tangent}
  n150 = Multiply(
   A=ComponentMask(None=@n149) {g=true, r=true}
   B=ScalarParameter() {defaultValue=0.8, group="-       EchoFrost Ice", parameterName=* Ice - Surface - Normal - Intensity, sortPriority=60}
  )
  n151 = ComponentMask(None=@n149) {b=true}
  n152 = MaterialFunctionCall(
   BaseNormal=MaterialFunctionCall(
    Normal=GetMaterialAttributes(None=@n146) {attributeGetTypes [#1]}.Normal
    Flatness=ScalarParameter() {defaultValue=0.75, group="-       EchoFrost Ice", parameterName=* Ice - Normal 4-Layer Blend Flatten, sortPriority=61}
   ) {materialFunction=&FlattenNormal}.Result
   AdditionalNormal=Normalize(VectorInput=AppendVector(A=@n150, B=@n151))
  ) {materialFunction=&BlendAngleCorrectedNormals}
  n153 = SetMaterialAttributes!(
   MaterialAttributes
   BaseColor=Multiply(
    A=Saturate(None=@n135)
    B=ScalarParameter() {defaultValue=1, group="-       EchoFrost Ice", parameterName=* Ice - Basecolor - Overall Brightness, sortPriority=50}
   )
   Metallic=Constant()
   Specular=Saturate(None=LinearInterpolate(A=@n136, B=@n140, Alpha=@n141))
   Roughness=LinearInterpolate(A=@n142, B=Multiply(A=@n144, B=@n142), Alpha=@n145)
   SubsurfaceColor
   Normal=@n152.Result
   Opacity
   EmissiveColor
   OpacityMask=GetMaterialAttributes(None=NamedRerouteUsage()) {attributeGetTypes [#8]}.OpacityMask
  ) {attributeSetTypes [
    #4
    #6
    #3
    #2
    #9
    #1
    #10
    #5
    #8
   ]}
  NamedRerouteDeclaration(None=@n153) {name=Ice, nodeColor=<rgba 1 0.4453 0 1>}
  NamedRerouteDeclaration(None=FunctionInput() {inputName=Baked AO, inputType=FunctionInput_Scalar, sortPriority=811}) {name=_Baked AO, nodeColor=<rgba 1 0 0.0525 1>}
  NamedRerouteDeclaration(None=FunctionInput() {inputName=DFAO, inputType=FunctionInput_Scalar, sortPriority=813}) {name=_DFAO, nodeColor=<rgba 1 0 0.0525 1>}
  NamedRerouteDeclaration(None=FunctionInput() {inputName=Baked Curvature, inputType=FunctionInput_Scalar, sortPriority=810}) {name=_Baked Curvature, nodeColor=<rgba 1 0 0.0525 1>}
  NamedRerouteDeclaration(None=FunctionInput() {inputName=Painted Mask, inputType=FunctionInput_Scalar, sortPriority=812}) {name=_Painted Mask, nodeColor=<rgba 1 0 0.0525 1>}
  NamedRerouteDeclaration(None=FunctionInput() {inputName=WS Noise Masks, inputType=FunctionInput_Scalar, sortPriority=815}) {name=_WS Noise Masks, nodeColor=<rgba 1 0 0.0525 1>}
  NamedRerouteDeclaration(None=FunctionInput() {bUsePreviewValueAsDefault=true, inputName=Tintable Mask, inputType=FunctionInput_Scalar, sortPriority=101}) {name=_Tintable Mask, nodeColor=<rgba 1 1 1 1>}
  n154 = Transform(None=@n152.Result)
  n155 = Add(
   A=Saturate(None=ComponentMask(None=@n154) {b=true})
   B=ScalarParameter() {group="-       EchoFrost Ice", parameterName=* Ice - Internal Light - Z up Normal light - Bias}
  )
  n156 = ScalarParameter() {defaultValue=1, group="-       EchoFrost Ice", parameterName=* Ice - Internal Light - Z up Normal light - Intensity}
  NamedRerouteDeclaration(None=Saturate(None=Multiply(A=@n155, B=@n156))) {name=Internal Light - Z up Normal Light, nodeColor=<rgba 1 0.8672 0 1>}
  NamedRerouteDeclaration(None=FunctionInput() {bUsePreviewValueAsDefault=true, inputName=EchoFrost Mask, inputType=FunctionInput_Scalar, sortPriority=100}) {name=EchoFrost Mask, nodeColor=<rgba 0.3229 0.5231 1 1>}
  NamedRerouteDeclaration(None=@n113) {name=WS Gradients, nodeColor=<rgba 0.4297 0 1 1>}
  n157 = ScalarParameter() {defaultValue=500, group="-       EchoFrost Ice", parameterName=* Internal Layers - Parallax - Scale 2, sortPriority=303}
  n158 = CameraVectorWS()
  n159 = Add(
   A=Divide(
    A=WorldPosition() {worldPositionShaderOffset=WPT_ExcludeAllShaderOffsets}.XYZ
    B=NamedRerouteDeclaration(None=@n157) {name=Offset UV Scale, nodeColor=<rgba 0.9688 1 0 1>}
   )
   B=Multiply(A=Multiply(A=@n158) {constB=-1}, B=NamedRerouteUsage())
  )
  n160 = ComponentMask(None=@n159) {g=true, r=true}
  n161 = ComponentMask(None=@n159) {b=true, r=true}
  n162 = NamedRerouteUsage()
  n163 = ComponentMask(None=@n162) {g=true}
  NamedRerouteDeclaration(None=LinearInterpolate(A=LinearInterpolate(A=@n160, B=@n161, Alpha=@n163), B=ComponentMask(None=@n159) {b=true, g=true}, Alpha=ComponentMask(None=@n162) {r=true})) {name=Offset UVs, nodeColor=<rgba 0.9453 0 1 1>}
  MaterialExpressionAbs_5 = Abs(None=VertexNormalWS())
  n164 = Convert(None=@MaterialExpressionAbs_5) {
   nodeName=Break Float3
   convertInputs [{
     type=Vector3
     expressionInput {
      expression=&self:MaterialExpressionAbs_5
      inputName=None
      mask=0
      maskA=0
      maskB=0
      maskG=0
      maskR=0
      outputIndex=0
     }
    }]
   convertOutputs [{type=Scalar}, {type=Scalar}, {type=Scalar}]
  }
  n165 = AppendVector(A=Subtract(A=@n164, B=Max(A=@n164, B=@n164)), B=Subtract(A=@n164, B=Max(A=@n164, B=@n164)))
  n166 = Subtract(A=@n164, B=Max(A=@n164, B=@n164))
  n167 = ScalarParameter() {defaultValue=0.1, group="-       EchoFrost Ice", parameterName=* Internal Layers - Parallax - Triplanar Blend, sortPriority=311}
  MaterialExpressionSaturate_21 = Saturate(None=Add(A=AppendVector(A=@n165, B=@n166), B=@n167) {constB=0.2})
  n168 = Convert(None=@MaterialExpressionSaturate_21) {
   nodeName=Break Float3
   convertInputs [{
     type=Vector3
     expressionInput {
      expression=&self:MaterialExpressionSaturate_21
      inputName=None
      mask=0
      maskA=0
      maskB=0
      maskG=0
      maskR=0
      outputIndex=0
     }
    }]
   convertOutputs [{type=Scalar}, {type=Scalar}, {type=Scalar}]
  }
  n169 = Divide(A=@MaterialExpressionSaturate_21, B=Add(A=Add(A=@n168, B=@n168), B=@n168))
  n170 = Add(A=ScreenPosition().PixelPosition, B=ViewProperty() {property=MEVP_TemporalSampleIndex}.Property)
  n171 = Add(A=Multiply(A=ComponentMask(None=@n170) {r=true}) {constB=0.2}, B=Multiply(A=ComponentMask(None=@n170) {g=true}) {constB=0.4})
  MaterialExpressionScreenPosition_5 = ScreenPosition()
  n172 = Custom!(PixelPosition=@MaterialExpressionScreenPosition_5.PixelPosition) {
   code="float4x4 bayerMatrix4x4 = float4x4(\r\n    0.0, 8.0, 2.0, 10.0,\r\n    12.0, 4.0, 14.0, 6.0,\r\n    3.0, 11.0, 1.0, 9.0,\r\n    15.0, 7.0, 13.0, 5.0\r\n) / 16.0f;\r\n\r\n    // Get the pixel's integer coordinates\r\n    int2 coords = Mod((int2)(PixelPosition.xy),4);\r\n\r\n    // HLSL accesses matrices using [row][column] syntax\r\n    float dither = bayerMatrix4x4[coords.y][coords.x];\r\n\r\n    return dither;"
   description=Bayer Matrix 4x4
   showCode=true
   inputs [{
     inputName=PixelPosition
     input {expression=&self:MaterialExpressionScreenPosition_5, outputIndex=1}
    }]
  }
  n173 = Add(A=Multiply(A=Frac(None=@n171)) {constB=2}, B=@n172) {constB=0.5}
  n174 = ComponentMask(None=Add(A=Multiply(A=@n173) {constB=0.3333}) {constB=-0.5}) {r=true}
  n175 = Multiply(A=@n174) {constB=-1}
  n176 = Round(None=Add(A=ComponentMask(None=@n169) {g=true, r=true}, B=AppendVector(A=@n174, B=@n175)) {constB=-0.5})
  n177 = Max(A=ComponentMask(None=@n176) {r=true}, B=ComponentMask(None=@n176) {g=true})
  NamedRerouteDeclaration(None=AppendVector(A=@n176, B=OneMinus(None=@n177))) {name=Dithered Mask, nodeColor=<rgba 1 0 0 1>}
  n178 = ScalarParameter() {defaultValue=0.2, group="-       EchoFrost Ice", parameterName=* Internal Layers - Parallax - DepthMinimum, sortPriority=304}
  n179 = ScalarParameter() {defaultValue=0.8, group="-       EchoFrost Ice", parameterName=* Internal Layers - Parallax - DepthMaximum, sortPriority=305}
  n180 = Divide(
   A=WorldPosition() {worldPositionShaderOffset=WPT_ExcludeAllShaderOffsets}.XYZ
   B=NamedRerouteDeclaration(None=ScalarParameter() {defaultValue=200, group="-       EchoFrost Ice", parameterName=* Internal Layers - Parallax - Scale 1, sortPriority=302}) {name=UV Scale, nodeColor=<rgba 0.9688 1 0 1>}
  )
  n181 = ComponentMask(None=@n180) {g=true, r=true}
  n182 = ComponentMask(None=@n180) {b=true, r=true}
  n183 = NamedRerouteUsage()
  n184 = ComponentMask(None=@n183) {g=true}
  n185 = Multiply(A=Divide(B=NamedRerouteUsage()) {constA=384}, B=TextureCoordinate())
  n186 = TextureSampleParameter2D!(
   UVs=LinearInterpolate(A=LinearInterpolate(A=@n181, B=@n182, Alpha=@n184), B=ComponentMask(None=@n180) {b=true, g=true}, Alpha=ComponentMask(None=@n183) {r=true})
   "DDX(UVs)"=DDX(Value=@n185)
   "DDY(UVs)"=DDY(Value=@n185)
   Apply View MipBias
  ) {
   group="-       EchoFrost Ice"
   mipValueMode=TMVM_Derivative
   parameterName=* Internal Layers - Parallax - DepthTexture
   samplerType=SAMPLERTYPE_Masks
   sortPriority=300
   texture=&T_EchoFrost_IceSurface_Facets_A_S
  }
  NamedRerouteDeclaration(None=LinearInterpolate(A=@n178, B=@n179, Alpha=@n186.R) {constA=-4, constB=-20}) {name=Depth Offset, nodeColor=<rgba 0.4531 1 0 1>}
  n187 = NamedRerouteUsage()
  NamedRerouteDeclaration(None=Multiply(A=Divide(B=@n187) {constA=384}, B=TextureCoordinate())) {name=Offset DDX DDY Coords, nodeColor=<rgba 1 0.2344 0 1>}
  n188 = WorldPosition()
  NamedRerouteDeclaration(None=ShadingPathSwitch(Default=@n188.Z, Mobile=TruncateLWC!(None=@n188.Z))) {name=WS Position Z, nodeColor=<rgba 0 0 1 1>}
  NamedRerouteDeclaration(None=MaterialFunctionCall(
    min=@n77
    max=@n76
    value=MaterialFunctionCall!(A, B=@n79.RGB, Bounds=@n81.RGB, Edge Falloff=@n83) {materialFunction=/Engine/Functions/Engine_MaterialFunctions01/Gradient/BoxMask-3D}.Result
   ) {materialFunction=&InverseLerp_MF}.Clamped) {name=Gradient A 3D, nodeColor=<rgba 0.1016 1 0 1>}
  NamedRerouteDeclaration(None=@n117.UVs) {name=Ice/Snow - UVs, nodeColor=<rgba 0.9219 0 1 1>}
  n189 = NamedRerouteUsage()
  n190 = TextureSampleParameter2D(UVs=NamedRerouteUsage(), "DDX(UVs)"=DDX(Value=@n189), "DDY(UVs)"=DDY(Value=@n189)) {
   group="-       EchoFrost Ice"
   mipValueMode=TMVM_Derivative
   parameterName=* Internal Layers - Parallax - Color Texture
   sortPriority=301
   texture=/Game/Global/Textures/Default/Masks_and_Gradients/T_MF_Crystal
  }
  n191 = ScalarParameter() {defaultValue=0.8, group="-       EchoFrost Ice", parameterName=* Internal Layers - Parallax - Color Desaturation, sortPriority=310}
  n192 = Power(
   Base=Desaturation(None=@n190.RGB, Fraction=@n191)
   Exp=ScalarParameter() {defaultValue=1, group="-       EchoFrost Ice", parameterName=* Internal Layers - Parallax - Power, sortPriority=308}
  )
  n193 = ScalarParameter() {defaultValue=-0.14, group="-       EchoFrost Ice", parameterName=* Internal Layers - Parallax - Bias, sortPriority=309}
  n194 = PixelDepth()
  n195 = Constant()
  n196 = ScalarParameter() {defaultValue=9000, group="-       EchoFrost Ice", parameterName=* Internal Layers - Parallax - Fade Distance, sortPriority=303}
  NamedRerouteDeclaration(None=Multiply(A=Add(A=@n192, B=@n193) {constB=-0.2}, B=SphereMask(A=@n194.R, B=@n195, Radius=@n196) {hardnessPercent=10})) {name=Parallax Layers, nodeColor=<rgba 0.5234 1 0 1>}
  NamedRerouteDeclaration(None=@n61.RGB) {name=_WS Noise Masks, nodeColor=<rgba 0 0 0.5469 1>}
  NamedRerouteUsage()
  NamedRerouteUsage()
  NamedRerouteUsage()
  NamedRerouteUsage()
  PixelNormalWS()
  Reroute(None=@n23)
  n197 = GetMaterialAttributes(None=@n55) {attributeGetTypes [
    #4
    #2
    #5
    #9
   ]}
  n198 = VectorParameter() {
   group="-       EchoFrost Ice"
   parameterName="EchoFrost - Scalability (Low)"
   sortPriority=0
   channelNames {a=SSS Multiply, b=Emissive Multiply, g=Roughness Multiply, r=BaseColor Multiply}
   defaultValue=<rgba 1 1 1 1>
  }
  n199 = VectorParameter() {
   group="-       EchoFrost Ice"
   parameterName="EchoFrost - Scalability (High)"
   sortPriority=0
   channelNames {a=SSS Multiply, b=Emissive Multiply, g=Roughness Multiply, r=BaseColor Multiply}
   defaultValue=<rgba 1 1 1 1>
  }
  n200 = VectorParameter() {
   group="-       EchoFrost Ice"
   parameterName="EchoFrost - Scalability (Medium)"
   sortPriority=0
   channelNames {a=SSS Multiply, b=Emissive Multiply, g=Roughness Multiply, r=BaseColor Multiply}
   defaultValue=<rgba 1 1 1 1>
  }
  n201 = SetMaterialAttributes!(
   MaterialAttributes=@n197.MaterialAttributes
   BaseColor=QualitySwitch(
    Default=@n197.BaseColor
    Low=Multiply(A=@n197.BaseColor, B=@n198.BaseColor Multiply)
    High=Multiply(A=@n197.BaseColor, B=@n199.BaseColor Multiply)
    Medium=Multiply(A=@n197.BaseColor, B=@n200.BaseColor Multiply)
    Epic=@n197.BaseColor
   )
   Roughness=QualitySwitch(
    Default=@n197.Roughness
    Low=Multiply(A=@n197.Roughness, B=@n198.Roughness Multiply)
    High=Multiply(A=@n197.Roughness, B=@n199.Roughness Multiply)
    Medium=Multiply(A=@n197.Roughness, B=@n200.Roughness Multiply)
    Epic=@n197.Roughness
   )
   EmissiveColor=QualitySwitch(
    Default=@n197.EmissiveColor
    Low=Multiply(A=@n197.EmissiveColor, B=@n198.Emissive Multiply)
    High=Multiply(A=@n197.EmissiveColor, B=@n199.Emissive Multiply)
    Medium=Multiply(A=@n197.EmissiveColor, B=@n200.Emissive Multiply)
    Epic=@n197.EmissiveColor
   )
   SubsurfaceColor=QualitySwitch(
    Default=@n197.SubsurfaceColor
    Low=Multiply(A=@n197.SubsurfaceColor, B=@n198.SSS Multiply)
    High=Multiply(A=@n197.SubsurfaceColor, B=@n199.SSS Multiply)
    Medium=Multiply(A=@n197.SubsurfaceColor, B=@n200.SSS Multiply)
    Epic=@n197.SubsurfaceColor
   )
  ) {attributeSetTypes [
    #4
    #2
    #5
    #9
   ]}
  Reroute(None=@n201)
  ShadingPathSwitch(
   Default=@n31
   Mobile=Add(A=@n30, B=ScalarParameter() {group="-       EchoFrost Snow Masking", parameterName="* Snow - Masking - Bias (Low/Mobile)", sortPriority=18})
  )
  ShadingPathSwitch()
 ]}
