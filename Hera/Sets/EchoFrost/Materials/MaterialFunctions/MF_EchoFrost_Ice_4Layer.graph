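# Material function that applies the switchable EchoFrost ice layer to material attributes, with shared Rock/Cliff settings.
# It wraps MF_EchoFrost_Architecture_Ice, feeding it an EchoFrost mask and curvature, AO and painted-mask data unpacked from material attributes.
# It computes distance-field AO that fades out with pixel depth, active only at Epic quality and constant on mobile.
# Tintable, Z-gradient and noise inputs are left unconnected on that call.
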MaterialFunction /Hera/Sets/EchoFrost/Materials/MaterialFunctions/MF_EchoFrost_Ice_4Layer
graph {nodes [
  FunctionInput() {inputName="\\ --- Masks --- /", inputType=FunctionInput_Scalar, sortPriority=20}
  FunctionInput() {inputName="\\ --- Dependencies --- /", inputType=FunctionInput_Scalar, sortPriority=800}
  n1 = GetMaterialAttributes(None=NamedRerouteUsage()) {attributeGetTypes [#1, #2, #3]}
  n2 = Constant() {r=1}
  n3 = ScalarParameter() {defaultValue=55, desc=55, group="--       Rock/Cliff Shared settings", parameterName=* DFAO - Distance, sortPriority=55}
  n4 = ScalarParameter() {defaultValue=150, desc=150, group="--       Rock/Cliff Shared settings", parameterName=* DFAO - Radius, sortPriority=56}
  n5 = DistanceFieldApproxAO!(World Position, Normal, BaseDistance=@n3, Radius=@n4) {numSteps=3}
  n6 = Divide(A=Subtract(A=PixelDepth().R, B=Constant() {r=5000}), B=Constant() {r=10000})
  n7 = Saturate(None=Subtract(A=OneMinus(None=@n5), B=Saturate(None=@n6)))
  n8 = Multiply(
   A=Power(
    Base=OneMinus(None=@n7)
    Exp=ScalarParameter() {defaultValue=1, group="--       Rock/Cliff Shared settings", parameterName=* DFAO - Power, sortPriority=52}
   )
   B=ScalarParameter() {defaultValue=1, group="--       Rock/Cliff Shared settings", parameterName=* DFAO - Multiply, sortPriority=51}
  )
  n9 = ShadingPathSwitch(Default=QualitySwitch(Default=@n2, Epic=Saturate(None=@n8)), Mobile=@n2)
  FunctionOutput(None=MaterialFunctionCall!(
    Use EchoFrost Ice Layer=StaticBoolParameter() {group="-       EchoFrost Ice", parameterName=Use EchoFrost Ice, sortPriority=-10}
    Input MaterialAttributes=@n1.MaterialAttributes
    "\\ --- Masks --- /"
    EchoFrost Mask=NamedRerouteUsage()
    Tintable Mask
    "\\ --- Dependencies --- /"
    Baked Curvature=@n1.MaterialAttributes
    Baked AO=@n1.MaterialAttributes
    Painted Mask=@n1.MaterialAttributes
    DFAO=NamedRerouteDeclaration(None=@n9) {name=_DFAO, nodeColor=<rgba 0 0 0.5469 1>}
    Z Gradient
    WS Noise Masks
   ) {materialFunction=&/MF_EchoFrost_Architecture_Ice}.MaterialAttributes)
  n10 = ScalarParameter() {
   defaultValue=1
   desc="Set to \"0\" if no AO texture is desired"
   group="--       Rock/Cliff Shared settings"
   parameterName=* DFAO - UseBakedAOTexture
   sortPriority=50
  }
  LinearInterpolate(A=Constant() {r=1}, Alpha=Saturate(None=@n10))
  NamedRerouteDeclaration(None=FunctionInput() {bUsePreviewValueAsDefault=true, inputName=EchoFrost Mask, inputType=FunctionInput_Scalar, sortPriority=100}) {name=EchoFrost Mask, nodeColor=<rgba 1 1 1 1>}
  NamedRerouteDeclaration(None=FunctionInput() {inputName=Z Gradient, inputType=FunctionInput_Scalar, sortPriority=814}) {name=_Z Gradient, nodeColor=<rgba 1 0 0.0525 1>}
  NamedRerouteDeclaration(None=FunctionInput(Preview=SetMaterialAttributes!(MaterialAttributes)) {inputName=Input MaterialAttributes, inputType=FunctionInput_MaterialAttributes}) {name=_Input MaterialAttributes, nodeColor=<rgba 0.1016 0 1 1>}
  NamedRerouteDeclaration(None=FunctionInput() {inputName=Baked AO, inputType=FunctionInput_Scalar, sortPriority=811}) {name=_Baked AO, nodeColor=<rgba 1 0 0.0525 1>}
  NamedRerouteDeclaration(None=FunctionInput() {inputName=DFAO, inputType=FunctionInput_Scalar, sortPriority=813}) {name=_DFAO, nodeColor=<rgba 1 0 0.0525 1>}
  NamedRerouteDeclaration(None=FunctionInput() {inputName=Baked Curvature, inputType=FunctionInput_Scalar, sortPriority=810}) {name=_Baked Curvature, nodeColor=<rgba 1 0 0.0525 1>}
  NamedRerouteDeclaration(None=FunctionInput() {inputName=Painted Mask, inputType=FunctionInput_Scalar, sortPriority=812}) {name=_Painted Mask, nodeColor=<rgba 1 0 0.0525 1>}
  NamedRerouteDeclaration(None=FunctionInput() {inputName=WS Noise Masks, inputType=FunctionInput_Scalar, sortPriority=815}) {name=_WS Noise Masks, nodeColor=<rgba 1 0 0.0525 1>}
  NamedRerouteDeclaration(None=FunctionInput() {bUsePreviewValueAsDefault=true, inputName=Tintable Mask, inputType=FunctionInput_Scalar, sortPriority=101}) {name=_Tintable Mask, nodeColor=<rgba 1 1 1 1>}
 ]}
