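# Material function that layers an optional EchoFrost frost coating over incoming material attributes, keeping their normal and opacity mask while setting color, roughness, specular and subsurface.
# Switchable world-Z normal, baked AO/curvature/painted and DFAO masks combine through min/max Ifs, then bias, intensity and invert.
# That mask drives a HeightLerp using a dithered-triplanar height texture, replaced by a cubed mask for material proxies.
# The result, scaled by opacity, sets the blend.
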MaterialFunction /Hera/Sets/EchoFrost/Materials/MaterialFunctions/MF_EchoFrost_Frost
graph {expressions {
  MaterialExpressionAdd_0 {
   class=MaterialExpressionAdd
   inputs [
    A {from=MaterialExpressionIf_9}
    B {from=MaterialExpressionNamedRerouteUsage_21}
   ]
  }
  MaterialExpressionAdd_3 {
   class=MaterialExpressionAdd
   inputs [
    A {from=MaterialExpressionIf_4}
    B {from=MaterialExpressionNamedRerouteUsage_86}
   ]
  }
  MaterialExpressionAdd_43 {
   class=MaterialExpressionAdd
   inputs [
    A {from=MaterialExpressionSaturate_17}
    B {from=MaterialExpressionNamedRerouteUsage_10}
   ]
  }
  MaterialExpressionAdd_44 {
   class=MaterialExpressionAdd
   inputs [
    A {from=MaterialExpressionMultiply_1}
    B {from=MaterialExpressionNamedRerouteUsage_31}
   ]
  }
  MaterialExpressionBlendMaterialAttributes_0 {
   class=MaterialExpressionBlendMaterialAttributes
   inputs [
    A {from=MaterialExpressionSetMaterialAttributes_3, output=""}
    B {from=MaterialExpressionNamedRerouteDeclaration_1, output=""}
    Alpha {from=MaterialExpressionReroute_3, output=""}
   ]
   properties {}
  }
  MaterialExpressionComponentMask_12 {class=MaterialExpressionComponentMask, inputs [None {from=MaterialExpressionStaticSwitch_0}], properties {b=true}}
  MaterialExpressionComponentMask_14 {class=MaterialExpressionComponentMask, inputs [None {from=MaterialExpressionVertexNormalWS_0}], properties {b=true}}
  MaterialExpressionConstant3Vector_1 {class=MaterialExpressionConstant3Vector, properties {constant=<rgba 0 0 1 0>}}
  MaterialExpressionConstant_0 {class=MaterialExpressionConstant}
  MaterialExpressionConstant_1 {class=MaterialExpressionConstant}
  MaterialExpressionConstant_2 {class=MaterialExpressionConstant, properties {r=3}}
  MaterialExpressionConstant_3 {class=MaterialExpressionConstant}
  MaterialExpressionConstant_6 {class=MaterialExpressionConstant}
  MaterialExpressionConvert_1 {
   class=MaterialExpressionConvert
   inputs [
    None {from=MaterialExpressionNamedRerouteUsage_90}
    None {from=MaterialExpressionNamedRerouteUsage_91}
    None {from=MaterialExpressionNamedRerouteUsage_92}
   ]
   properties {
    nodeName=Make Vector3
    convertInputs [
     {
      type=Scalar
      expressionInput {
       expression=&self:MaterialExpressionNamedRerouteUsage_90
       inputName=None
       mask=0
       maskA=0
       maskB=0
       maskG=0
       maskR=0
       outputIndex=0
      }
     }
     {
      type=Scalar
      expressionInput {
       expression=&self:MaterialExpressionNamedRerouteUsage_91
       inputName=None
       mask=0
       maskA=0
       maskB=0
       maskG=0
       maskR=0
       outputIndex=0
      }
     }
     {
      type=Scalar
      expressionInput {
       expression=&self:MaterialExpressionNamedRerouteUsage_92
       inputName=None
       mask=0
       maskA=0
       maskB=0
       maskG=0
       maskR=0
       outputIndex=0
      }
     }
    ]
    convertOutputs [{type=Vector3}]
   }
  }
  MaterialExpressionDotProduct_2 {
   class=MaterialExpressionDotProduct
   inputs [
    A {from=MaterialExpressionNamedRerouteUsage_13, output=""}
    B {from=MaterialExpressionNamedRerouteUsage_5, output=""}
   ]
   properties {}
  }
  MaterialExpressionFunctionInput_1 {
   class=MaterialExpressionFunctionInput
   inputs [Preview {from=MaterialExpressionStaticBool_0, output=""}]
   properties {bUsePreviewValueAsDefault=true, inputName=Use EchoFrost Frost, inputType=FunctionInput_StaticBool}
  }
  MaterialExpressionFunctionInput_10 {
   class=MaterialExpressionFunctionInput
   inputs [Preview {from="", output=""}]
   properties {inputName=DFAO, inputType=FunctionInput_Scalar, sortPriority=813}
  }
  MaterialExpressionFunctionInput_11 {
   class=MaterialExpressionFunctionInput
   inputs [Preview {from=MaterialExpressionSetMaterialAttributes_0, output=""}]
   properties {
    description=MaterialAttributes from previous layers
    inputName=Input MaterialAttributes
    inputType=FunctionInput_MaterialAttributes
    sortPriority=-100
   }
  }
  MaterialExpressionFunctionInput_12 {
   class=MaterialExpressionFunctionInput
   inputs [Preview {from="", output=""}]
   properties {inputName=Baked Curvature, inputType=FunctionInput_Scalar, sortPriority=810}
  }
  MaterialExpressionFunctionInput_14 {
   class=MaterialExpressionFunctionInput
   inputs [Preview {from="", output=""}]
   properties {inputName=Z Gradient, inputType=FunctionInput_Scalar, sortPriority=814}
  }
  MaterialExpressionFunctionInput_15 {
   class=MaterialExpressionFunctionInput
   inputs [Preview {from="", output=""}]
   properties {inputName=Baked AO, inputType=FunctionInput_Scalar, sortPriority=811}
  }
  MaterialExpressionFunctionInput_16 {
   class=MaterialExpressionFunctionInput
   inputs [Preview {from="", output=""}]
   properties {inputName=Painted Mask, inputType=FunctionInput_Scalar, sortPriority=812}
  }
  MaterialExpressionFunctionInput_26 {class=MaterialExpressionFunctionInput, inputs [Preview {from="", output=""}], properties {inputName=WS Noise Masks, sortPriority=815}}
  MaterialExpressionFunctionInput_8 {
   class=MaterialExpressionFunctionInput
   inputs [Preview {from="", output=""}]
   properties {inputName="\\ --- Dependencies --- /", inputType=FunctionInput_Scalar, sortPriority=800}
  }
  MaterialExpressionFunctionOutput_2 {
   class=MaterialExpressionFunctionOutput
   inputs [None {from=MaterialExpressionStaticSwitch_4, output=""}]
   properties {outputName=Blended MaterialAttributes, sortPriority=-100}
  }
  MaterialExpressionGetMaterialAttributes_0 {
   class=MaterialExpressionGetMaterialAttributes
   inputs [None {from=MaterialExpressionNamedRerouteUsage_4, output=""}]
   properties {attributeGetTypes [#1, #2, #3]}
  }
  MaterialExpressionGetMaterialAttributes_16 {
   class=MaterialExpressionGetMaterialAttributes
   inputs [None {from=MaterialExpressionNamedRerouteUsage_1, output=""}]
   properties {attributeGetTypes [#4]}
  }
  MaterialExpressionGetMaterialAttributes_8 {
   class=MaterialExpressionGetMaterialAttributes
   inputs [None {from=MaterialExpressionNamedRerouteUsage_8, output=""}]
   properties {attributeGetTypes [#5, #4]}
  }
  MaterialExpressionIf_3 {
   class=MaterialExpressionIf
   inputs [
    A {from=MaterialExpressionNamedRerouteUsage_38, output=""}
    B {from="", output=""}
    "A > B" {from=MaterialExpressionMax_19, output=""}
    "A == B" {from="", output=""}
    "A < B" {from=MaterialExpressionMin_18, output=""}
   ]
   properties {constB=0.5, desc="cheaper than lerp without \"==\" condition", equalsThreshold=0.001}
  }
  MaterialExpressionIf_4 {
   class=MaterialExpressionIf
   inputs [
    A {from=MaterialExpressionNamedRerouteUsage_96, output=""}
    B {from="", output=""}
    "A > B" {from=MaterialExpressionOneMinus_4, output=""}
    "A == B" {from="", output=""}
    "A < B" {from=MaterialExpressionDotProduct_2, output=""}
   ]
   properties {constB=0.5, desc="cheaper than lerp without \"==\" condition", equalsThreshold=0.001}
  }
  MaterialExpressionIf_5 {
   class=MaterialExpressionIf
   inputs [
    A {from=MaterialExpressionNamedRerouteUsage_33, output=""}
    B {from="", output=""}
    "A > B" {from=MaterialExpressionMax_20, output=""}
    "A == B" {from="", output=""}
    "A < B" {from=MaterialExpressionMin_19, output=""}
   ]
   properties {constB=0.5, desc="cheaper than lerp without \"==\" condition", equalsThreshold=0.001}
  }
  MaterialExpressionIf_9 {
   class=MaterialExpressionIf
   inputs [
    A {from=MaterialExpressionNamedRerouteUsage_47, output=""}
    B {from="", output=""}
    "A > B" {from=MaterialExpressionOneMinus_31, output=""}
    "A == B" {from="", output=""}
    "A < B" {from=MaterialExpressionReroute_235, output=""}
   ]
   properties {constB=0.5, desc="cheaper than lerp without \"==\" condition", equalsThreshold=0.001}
  }
  MaterialExpressionLinearInterpolate_0 {
   class=MaterialExpressionLinearInterpolate
   inputs [
    A {from=MaterialExpressionGetMaterialAttributes_0, output=EmissiveColor}
    B {from=MaterialExpressionConstant_1}
    Alpha {from=MaterialExpressionReroute_2}
   ]
  }
  MaterialExpressionLinearInterpolate_1 {
   class=MaterialExpressionLinearInterpolate
   inputs [
    A {from=MaterialExpressionScalarParameter_1}
    B {from=MaterialExpressionScalarParameter_3}
    Alpha {from=MaterialExpressionTextureSample_13, output=R}
   ]
  }
  MaterialExpressionMaterialFunctionCall_17 {
   class=MaterialExpressionMaterialFunctionCall
   inputs [
    A {from=""}
    B {from=""}
    Transition Phase {from=MaterialExpressionNamedRerouteUsage_3}
    Height Texture {from=MaterialExpressionLinearInterpolate_1}
    Contrast {from=MaterialExpressionScalarParameter_50}
   ]
   properties {materialFunction=&HeightLerp}
  }
  MaterialExpressionMaterialFunctionCall_30 {
   class=MaterialExpressionMaterialFunctionCall
   inputs [
    Absolute World Position {from=""}
    "Normal (WS)" {from=""}
    Texture Scale {from=MaterialExpressionScalarParameter_69}
    Triplanar Fade {from=""}
    Derivatives Scale {from=""}
   ]
   properties {materialFunction=/BRCoreMaterials/Materials/MaterialFunctions/MF_TriplanarDithered_WavyDistortion}
  }
  MaterialExpressionMaterialProxyReplace_0 {
   class=MaterialExpressionMaterialProxyReplace
   inputs [
    Realtime {from=MaterialExpressionMaterialFunctionCall_17, output=Lerp Alpha No Contrast}
    MaterialProxy {from=MaterialExpressionSaturate_0, output=""}
   ]
   properties {}
  }
  MaterialExpressionMax_19 {
   class=MaterialExpressionMax
   inputs [
    A {from=MaterialExpressionReroute_237}
    B {from=MaterialExpressionReroute_239}
   ]
  }
  MaterialExpressionMax_20 {
   class=MaterialExpressionMax
   inputs [
    A {from=MaterialExpressionReroute_241}
    B {from=MaterialExpressionReroute_242}
   ]
  }
  MaterialExpressionMin_18 {
   class=MaterialExpressionMin
   inputs [
    A {from=MaterialExpressionReroute_237, output=""}
    B {from=MaterialExpressionReroute_239, output=""}
   ]
   properties {}
  }
  MaterialExpressionMin_19 {
   class=MaterialExpressionMin
   inputs [
    A {from=MaterialExpressionReroute_241, output=""}
    B {from=MaterialExpressionReroute_242, output=""}
   ]
   properties {}
  }
  MaterialExpressionMultiply_0 {
   class=MaterialExpressionMultiply
   inputs [
    A {from=MaterialExpressionAdd_0}
    B {from=MaterialExpressionNamedRerouteUsage_19}
   ]
  }
  MaterialExpressionMultiply_1 {
   class=MaterialExpressionMultiply
   inputs [
    A {from=MaterialExpressionStaticSwitch_7}
    B {from=""}
   ]
   properties {constB=-1}
  }
  MaterialExpressionMultiply_2 {
   class=MaterialExpressionMultiply
   inputs [
    A {from=MaterialExpressionMultiply_71}
    B {from=MaterialExpressionNamedRerouteUsage_14}
   ]
  }
  MaterialExpressionMultiply_3 {
   class=MaterialExpressionMultiply
   inputs [
    A {from=MaterialExpressionVectorParameter_16, output=RGB}
    B {from=MaterialExpressionScalarParameter_0}
   ]
  }
  MaterialExpressionMultiply_57 {
   class=MaterialExpressionMultiply
   inputs [
    A {from=MaterialExpressionReroute_1}
    B {from=MaterialExpressionScalarParameter_77}
   ]
   properties {constB=0.1}
  }
  MaterialExpressionMultiply_69 {
   class=MaterialExpressionMultiply
   inputs [
    A {from=MaterialExpressionAdd_43}
    B {from=MaterialExpressionNamedRerouteUsage_11}
   ]
  }
  MaterialExpressionMultiply_71 {
   class=MaterialExpressionMultiply
   inputs [
    A {from=MaterialExpressionAdd_3}
    B {from=""}
   ]
   properties {constB=3}
  }
  MaterialExpressionMultiply_72 {
   class=MaterialExpressionMultiply
   inputs [
    A {from=MaterialExpressionAdd_44}
    B {from=MaterialExpressionNamedRerouteUsage_32}
   ]
  }
  MaterialExpressionNamedRerouteDeclaration_0 {
   class=MaterialExpressionNamedRerouteDeclaration
   inputs [None {from=MaterialExpressionStaticBoolParameter_43}]
   properties {name=--- Use DFAO MASKING ---, nodeColor=<rgba 0.1979 0.0084 0 1>}
  }
  MaterialExpressionNamedRerouteDeclaration_1 {
   class=MaterialExpressionNamedRerouteDeclaration
   inputs [None {from=MaterialExpressionSetMaterialAttributes_6}]
   properties {name=Frost - MaterialAttributes, nodeColor=<rgba 0 1 0.8125 1>}
  }
  MaterialExpressionNamedRerouteDeclaration_11 {
   class=MaterialExpressionNamedRerouteDeclaration
   inputs [None {from=MaterialExpressionFunctionInput_10}]
   properties {name=_DFAO, nodeColor=<rgba 1 0 0.0525 1>}
  }
  MaterialExpressionNamedRerouteDeclaration_12 {
   class=MaterialExpressionNamedRerouteDeclaration
   inputs [None {from=MaterialExpressionFunctionInput_12}]
   properties {name=_Baked Curvature, nodeColor=<rgba 1 0 0.0525 1>}
  }
  MaterialExpressionNamedRerouteDeclaration_13 {
   class=MaterialExpressionNamedRerouteDeclaration
   inputs [None {from=MaterialExpressionFunctionInput_16}]
   properties {name=_Painted Mask, nodeColor=<rgba 1 0 0.0525 1>}
  }
  MaterialExpressionNamedRerouteDeclaration_2 {
   class=MaterialExpressionNamedRerouteDeclaration
   inputs [None {from=MaterialExpressionStaticBoolParameter_0}]
   properties {name="• World Z Normal - Input Uses Tangent Space Normals", nodeColor=<rgba 0.1979 0.0084 0 1>}
  }
  MaterialExpressionNamedRerouteDeclaration_22 {
   class=MaterialExpressionNamedRerouteDeclaration
   inputs [None {from=MaterialExpressionFunctionInput_26}]
   properties {name=_WS Noise Masks, nodeColor=<rgba 1 0 0.0525 1>}
  }
  MaterialExpressionNamedRerouteDeclaration_23 {
   class=MaterialExpressionNamedRerouteDeclaration
   inputs [None {from=MaterialExpressionConvert_1}]
   properties {name=Baked AO/Curvature/Painted Texture, nodeColor=<rgba 1 1 1 1>}
  }
  MaterialExpressionNamedRerouteDeclaration_26 {
   class=MaterialExpressionNamedRerouteDeclaration
   inputs [None {from=MaterialExpressionScalarParameter_21}]
   properties {name="• AO/CURVATURE/THICKNESS - Invert", nodeColor=<rgba 0.5043 1 0 1>}
  }
  MaterialExpressionNamedRerouteDeclaration_27 {
   class=MaterialExpressionNamedRerouteDeclaration
   inputs [None {from=MaterialExpressionScalarParameter_92}]
   properties {name=* World Z Normal - Bias, nodeColor=<rgba 0.5043 1 0 1>}
  }
  MaterialExpressionNamedRerouteDeclaration_28 {
   class=MaterialExpressionNamedRerouteDeclaration
   inputs [None {from=MaterialExpressionStaticBoolParameter_31}]
   properties {name=--- Use WORLD Z NORMAL MASKING ---, nodeColor=<rgba 0.1979 0.0084 0 1>}
  }
  MaterialExpressionNamedRerouteDeclaration_29 {
   class=MaterialExpressionNamedRerouteDeclaration
   inputs [None {from=MaterialExpressionScalarParameter_93}]
   properties {name=* World Z Normal - Intensity, nodeColor=<rgba 0.5043 1 0 1>}
  }
  MaterialExpressionNamedRerouteDeclaration_30 {
   class=MaterialExpressionNamedRerouteDeclaration
   inputs [None {from=MaterialExpressionStaticBoolParameter_11}]
   properties {name="• World Z Normal - Use VTX Normals", desc="T= VTX Normals only\r\nF= Combined Material Normals", nodeColor=<rgba 0.1979 0.0084 0 1>}
  }
  MaterialExpressionNamedRerouteDeclaration_31 {
   class=MaterialExpressionNamedRerouteDeclaration
   inputs [None {from=MaterialExpressionScalarParameter_2}]
   properties {name="• AO/CURVATURE/THICKNESS - Intensity", nodeColor=<rgba 0.5043 1 0 1>}
  }
  MaterialExpressionNamedRerouteDeclaration_32 {
   class=MaterialExpressionNamedRerouteDeclaration
   inputs [None {from=MaterialExpressionStaticBoolParameter_42}]
   properties {name=--- Use AO/CURVATURE/THICKNESS MASKING ---, nodeColor=<rgba 0.1979 0.0084 0 1>}
  }
  MaterialExpressionNamedRerouteDeclaration_33 {
   class=MaterialExpressionNamedRerouteDeclaration
   inputs [None {from=MaterialExpressionScalarParameter_11}]
   properties {name="• AO/CURVATURE/THICKNESS - Bias", nodeColor=<rgba 0.5043 1 0 1>}
  }
  MaterialExpressionNamedRerouteDeclaration_34 {
   class=MaterialExpressionNamedRerouteDeclaration
   inputs [None {from=MaterialExpressionVectorParameter_0, output=RGB}]
   properties {name="• AO/CURVATURE/THICKNESS - Channel Weights", nodeColor=<rgba 0.5043 1 0 1>}
  }
  MaterialExpressionNamedRerouteDeclaration_36 {
   class=MaterialExpressionNamedRerouteDeclaration
   inputs [None {from=MaterialExpressionScalarParameter_18}]
   properties {name=* Masking - Invert, nodeColor=<rgba 0.5043 1 0 1>}
  }
  MaterialExpressionNamedRerouteDeclaration_38 {
   class=MaterialExpressionNamedRerouteDeclaration
   inputs [None {from=MaterialExpressionScalarParameter_7}]
   properties {name="• Masking - Bias", nodeColor=<rgba 0.5043 1 0 1>}
  }
  MaterialExpressionNamedRerouteDeclaration_39 {
   class=MaterialExpressionNamedRerouteDeclaration
   inputs [None {from=MaterialExpressionScalarParameter_8}]
   properties {name="• Masking - Intensity", nodeColor=<rgba 0.5043 1 0 1>}
  }
  MaterialExpressionNamedRerouteDeclaration_48 {
   class=MaterialExpressionNamedRerouteDeclaration
   inputs [None {from=MaterialExpressionScalarParameter_95}]
   properties {name="• DFAO - Intensity", nodeColor=<rgba 0.5043 1 0 1>}
  }
  MaterialExpressionNamedRerouteDeclaration_49 {
   class=MaterialExpressionNamedRerouteDeclaration
   inputs [None {from=MaterialExpressionScalarParameter_96}]
   properties {name="• DFAO - Bias", nodeColor=<rgba 0.5043 1 0 1>}
  }
  MaterialExpressionNamedRerouteDeclaration_51 {
   class=MaterialExpressionNamedRerouteDeclaration
   inputs [None {from=MaterialExpressionReroute_20}]
   properties {name=Combined Mask, nodeColor=<rgba 0.5 0 1 1>}
  }
  MaterialExpressionNamedRerouteDeclaration_53 {
   class=MaterialExpressionNamedRerouteDeclaration
   inputs [None {from=MaterialExpressionStaticBoolParameter_25}]
   properties {name=* DFAO - Force Baked AO on Nanite, nodeColor=<rgba 0.1979 0.0084 0 1>}
  }
  MaterialExpressionNamedRerouteDeclaration_7 {
   class=MaterialExpressionNamedRerouteDeclaration
   inputs [None {from=MaterialExpressionFunctionInput_14}]
   properties {name=_Z Gradient, nodeColor=<rgba 1 0 0.0525 1>}
  }
  MaterialExpressionNamedRerouteDeclaration_74 {
   class=MaterialExpressionNamedRerouteDeclaration
   inputs [None {from=MaterialExpressionScalarParameter_15}]
   properties {name="• AO/CURVATURE/THICKNESS - Blend (1=Max 0=Min)", nodeColor=<rgba 0.5043 1 0 1>}
  }
  MaterialExpressionNamedRerouteDeclaration_75 {
   class=MaterialExpressionNamedRerouteDeclaration
   inputs [None {from=MaterialExpressionScalarParameter_16}]
   properties {name="• DFAO - Blend (1=Max 0=Min)", nodeColor=<rgba 0.5043 1 0 1>}
  }
  MaterialExpressionNamedRerouteDeclaration_8 {
   class=MaterialExpressionNamedRerouteDeclaration
   inputs [None {from=MaterialExpressionFunctionInput_15}]
   properties {name=_Baked AO, nodeColor=<rgba 1 0 0.0525 1>}
  }
  MaterialExpressionNamedRerouteDeclaration_83 {
   class=MaterialExpressionNamedRerouteDeclaration
   inputs [None {from=MaterialExpressionSaturate_32}]
   properties {name=Normal Z, nodeColor=<rgba 0.5 0 1 1>}
  }
  MaterialExpressionNamedRerouteDeclaration_84 {
   class=MaterialExpressionNamedRerouteDeclaration
   inputs [None {from=MaterialExpressionSaturate_33}]
   properties {name=AO/CURVATURE/THICKNESS, nodeColor=<rgba 0.5 0 1 1>}
  }
  MaterialExpressionNamedRerouteDeclaration_85 {
   class=MaterialExpressionNamedRerouteDeclaration
   inputs [None {from=MaterialExpressionSaturate_34}]
   properties {name=DFAO - Mask, nodeColor=<rgba 0.5 0 1 1>}
  }
  MaterialExpressionNamedRerouteDeclaration_9 {
   class=MaterialExpressionNamedRerouteDeclaration
   inputs [None {from=MaterialExpressionReroute_15}]
   properties {name=_Input A MaterialAttributes, desc=MaterialAttributes from previous layers, nodeColor=<rgba 0.1016 0 1 1>}
  }
  MaterialExpressionNamedRerouteUsage_0 {class=MaterialExpressionNamedRerouteUsage}
  MaterialExpressionNamedRerouteUsage_1 {class=MaterialExpressionNamedRerouteUsage}
  MaterialExpressionNamedRerouteUsage_10 {class=MaterialExpressionNamedRerouteUsage}
  MaterialExpressionNamedRerouteUsage_11 {class=MaterialExpressionNamedRerouteUsage}
  MaterialExpressionNamedRerouteUsage_12 {class=MaterialExpressionNamedRerouteUsage}
  MaterialExpressionNamedRerouteUsage_13 {class=MaterialExpressionNamedRerouteUsage}
  MaterialExpressionNamedRerouteUsage_14 {class=MaterialExpressionNamedRerouteUsage, properties {desc=Modulates VTX color transition intensity}}
  MaterialExpressionNamedRerouteUsage_15 {class=MaterialExpressionNamedRerouteUsage}
  MaterialExpressionNamedRerouteUsage_170 {class=MaterialExpressionNamedRerouteUsage}
  MaterialExpressionNamedRerouteUsage_171 {class=MaterialExpressionNamedRerouteUsage}
  MaterialExpressionNamedRerouteUsage_19 {class=MaterialExpressionNamedRerouteUsage}
  MaterialExpressionNamedRerouteUsage_2 {class=MaterialExpressionNamedRerouteUsage}
  MaterialExpressionNamedRerouteUsage_21 {class=MaterialExpressionNamedRerouteUsage}
  MaterialExpressionNamedRerouteUsage_3 {class=MaterialExpressionNamedRerouteUsage}
  MaterialExpressionNamedRerouteUsage_31 {class=MaterialExpressionNamedRerouteUsage}
  MaterialExpressionNamedRerouteUsage_32 {class=MaterialExpressionNamedRerouteUsage}
  MaterialExpressionNamedRerouteUsage_33 {class=MaterialExpressionNamedRerouteUsage}
  MaterialExpressionNamedRerouteUsage_38 {class=MaterialExpressionNamedRerouteUsage}
  MaterialExpressionNamedRerouteUsage_4 {class=MaterialExpressionNamedRerouteUsage}
  MaterialExpressionNamedRerouteUsage_46 {class=MaterialExpressionNamedRerouteUsage}
  MaterialExpressionNamedRerouteUsage_47 {class=MaterialExpressionNamedRerouteUsage}
  MaterialExpressionNamedRerouteUsage_5 {class=MaterialExpressionNamedRerouteUsage}
  MaterialExpressionNamedRerouteUsage_6 {class=MaterialExpressionNamedRerouteUsage}
  MaterialExpressionNamedRerouteUsage_7 {class=MaterialExpressionNamedRerouteUsage}
  MaterialExpressionNamedRerouteUsage_8 {class=MaterialExpressionNamedRerouteUsage}
  MaterialExpressionNamedRerouteUsage_85 {class=MaterialExpressionNamedRerouteUsage}
  MaterialExpressionNamedRerouteUsage_86 {class=MaterialExpressionNamedRerouteUsage}
  MaterialExpressionNamedRerouteUsage_9 {class=MaterialExpressionNamedRerouteUsage}
  MaterialExpressionNamedRerouteUsage_90 {class=MaterialExpressionNamedRerouteUsage}
  MaterialExpressionNamedRerouteUsage_91 {class=MaterialExpressionNamedRerouteUsage}
  MaterialExpressionNamedRerouteUsage_92 {class=MaterialExpressionNamedRerouteUsage}
  MaterialExpressionNamedRerouteUsage_96 {class=MaterialExpressionNamedRerouteUsage}
  MaterialExpressionOneMinus_31 {class=MaterialExpressionOneMinus, inputs [None {from=MaterialExpressionReroute_235}]}
  MaterialExpressionOneMinus_4 {class=MaterialExpressionOneMinus, inputs [None {from=MaterialExpressionDotProduct_2}]}
  MaterialExpressionOneMinus_7 {class=MaterialExpressionOneMinus, inputs [None {from=MaterialExpressionTextureSample_13, output=R}]}
  MaterialExpressionPower_0 {
   class=MaterialExpressionPower
   inputs [
    Base {from=MaterialExpressionNamedRerouteUsage_3}
    Exp {from=MaterialExpressionConstant_2}
   ]
  }
  MaterialExpressionPower_1 {
   class=MaterialExpressionPower
   inputs [
    Base {from=MaterialExpressionNamedRerouteUsage_3}
    Exp {from=""}
   ]
  }
  MaterialExpressionReroute_0 {class=MaterialExpressionReroute, inputs [None {from=MaterialExpressionStaticSwitch_3}]}
  MaterialExpressionReroute_1 {class=MaterialExpressionReroute, inputs [None {from=MaterialExpressionMaterialProxyReplace_0}]}
  MaterialExpressionReroute_15 {class=MaterialExpressionReroute, inputs [None {from=MaterialExpressionFunctionInput_11}]}
  MaterialExpressionReroute_19 {class=MaterialExpressionReroute, inputs [None {from=MaterialExpressionStaticSwitch_1}]}
  MaterialExpressionReroute_2 {class=MaterialExpressionReroute, inputs [None {from=MaterialExpressionReroute_1}]}
  MaterialExpressionReroute_20 {class=MaterialExpressionReroute, inputs [None {from=MaterialExpressionSaturate_39}]}
  MaterialExpressionReroute_228 {class=MaterialExpressionReroute, inputs [None {from=MaterialExpressionReroute_241}]}
  MaterialExpressionReroute_230 {class=MaterialExpressionReroute, inputs [None {from=MaterialExpressionReroute_237}]}
  MaterialExpressionReroute_235 {class=MaterialExpressionReroute, inputs [None {from=MaterialExpressionReroute_25}]}
  MaterialExpressionReroute_237 {class=MaterialExpressionReroute, inputs [None {from=MaterialExpressionReroute_19}]}
  MaterialExpressionReroute_239 {class=MaterialExpressionReroute, inputs [None {from=MaterialExpressionNamedRerouteUsage_171}]}
  MaterialExpressionReroute_241 {class=MaterialExpressionReroute, inputs [None {from=MaterialExpressionReroute_0}]}
  MaterialExpressionReroute_242 {class=MaterialExpressionReroute, inputs [None {from=MaterialExpressionNamedRerouteUsage_170}]}
  MaterialExpressionReroute_25 {class=MaterialExpressionReroute, inputs [None {from=MaterialExpressionStaticSwitch_2}]}
  MaterialExpressionReroute_3 {class=MaterialExpressionReroute, inputs [None {from=MaterialExpressionMultiply_57}]}
  MaterialExpressionReroute_34 {class=MaterialExpressionReroute, inputs [None {from=MaterialExpressionGetMaterialAttributes_8, output=OpacityMask}]}
  MaterialExpressionReroute_4 {class=MaterialExpressionReroute, inputs [None {from=MaterialExpressionGetMaterialAttributes_8, output=Normal}]}
  MaterialExpressionSaturate_0 {class=MaterialExpressionSaturate, inputs [None {from=MaterialExpressionPower_0}]}
  MaterialExpressionSaturate_1 {class=MaterialExpressionSaturate, inputs [None {from=MaterialExpressionComponentMask_14}]}
  MaterialExpressionSaturate_17 {class=MaterialExpressionSaturate, inputs [None {from=MaterialExpressionComponentMask_12}]}
  MaterialExpressionSaturate_32 {class=MaterialExpressionSaturate, inputs [None {from=MaterialExpressionMultiply_69}]}
  MaterialExpressionSaturate_33 {class=MaterialExpressionSaturate, inputs [None {from=MaterialExpressionMultiply_2}]}
  MaterialExpressionSaturate_34 {class=MaterialExpressionSaturate, inputs [None {from=MaterialExpressionMultiply_72}]}
  MaterialExpressionSaturate_39 {class=MaterialExpressionSaturate, inputs [None {from=MaterialExpressionMultiply_0}]}
  MaterialExpressionScalarParameter_0 {class=MaterialExpressionScalarParameter, properties {defaultValue=0.5, group="-       EchoFrost Frost", parameterName=* Frost - SSS Color}}
  MaterialExpressionScalarParameter_1 {class=MaterialExpressionScalarParameter, properties {group="-       EchoFrost Frost", parameterName=* Frost - Height - Min}}
  MaterialExpressionScalarParameter_11 {
   class=MaterialExpressionScalarParameter
   properties {group="-       EchoFrost Frost", parameterName="• Frost - AO/Curvature/Thickness - Bias", sortPriority=175}
  }
  MaterialExpressionScalarParameter_15 {
   class=MaterialExpressionScalarParameter
   properties {defaultValue=1, group="-       EchoFrost Frost", parameterName="• Frost - AO/Curvature/Thickness - Blend (1=Max 0=Min)", sortPriority=170}
  }
  MaterialExpressionScalarParameter_16 {
   class=MaterialExpressionScalarParameter
   properties {defaultValue=1, group="-       EchoFrost Frost", parameterName="• Frost - DFAO - Blend (1=Max 0=Min)", sortPriority=220}
  }
  MaterialExpressionScalarParameter_18 {
   class=MaterialExpressionScalarParameter
   properties {group="-       EchoFrost Frost", parameterName="• Frost - Masking - Invert", sliderMax=1, sortPriority=50}
  }
  MaterialExpressionScalarParameter_2 {
   class=MaterialExpressionScalarParameter
   properties {defaultValue=1, group="-       EchoFrost Frost", parameterName="• Frost - AO/Curvature/Thickness - Intensity", sortPriority=176}
  }
  MaterialExpressionScalarParameter_21 {
   class=MaterialExpressionScalarParameter
   properties {defaultValue=1, group="-       EchoFrost Frost", parameterName="• Frost - AO/Curvature/Thickness - Invert", sortPriority=161}
  }
  MaterialExpressionScalarParameter_3 {class=MaterialExpressionScalarParameter, properties {defaultValue=1, group="-       EchoFrost Frost", parameterName=* Frost - Height - Max}}
  MaterialExpressionScalarParameter_46 {class=MaterialExpressionScalarParameter, properties {defaultValue=-0.5, group="-       EchoFrost Frost", parameterName=* Frost - Specular}}
  MaterialExpressionScalarParameter_47 {class=MaterialExpressionScalarParameter, properties {defaultValue=2, group="-       EchoFrost Frost", parameterName=* Frost - Roughness}}
  MaterialExpressionScalarParameter_48 {class=MaterialExpressionScalarParameter, properties {defaultValue=0.5, group="-       EchoFrost Frost", parameterName=* Frost - SSS Opacity}}
  MaterialExpressionScalarParameter_50 {class=MaterialExpressionScalarParameter, properties {group="-       EchoFrost Frost", parameterName=* Frost - Contrast}}
  MaterialExpressionScalarParameter_69 {class=MaterialExpressionScalarParameter, properties {defaultValue=772.0963, group="-       EchoFrost Frost", parameterName=* Frost - Scale}}
  MaterialExpressionScalarParameter_7 {
   class=MaterialExpressionScalarParameter
   properties {group="-       EchoFrost Frost", parameterName="• Frost - Masking - Bias", sortPriority=60}
  }
  MaterialExpressionScalarParameter_77 {class=MaterialExpressionScalarParameter, properties {defaultValue=0.12, group="-       EchoFrost Frost", parameterName=* Frost - Opacity}}
  MaterialExpressionScalarParameter_8 {
   class=MaterialExpressionScalarParameter
   properties {defaultValue=1, group="-       EchoFrost Frost", parameterName="• Frost - Masking - Intensity", sortPriority=61}
  }
  MaterialExpressionScalarParameter_92 {
   class=MaterialExpressionScalarParameter
   properties {group="-       EchoFrost Frost", parameterName="• Frost - World Z Normal - Bias", sortPriority=120}
  }
  MaterialExpressionScalarParameter_93 {
   class=MaterialExpressionScalarParameter
   properties {defaultValue=1, group="-       EchoFrost Frost", parameterName="• Frost - World Z Normal - Intensity", sortPriority=121}
  }
  MaterialExpressionScalarParameter_95 {
   class=MaterialExpressionScalarParameter
   properties {defaultValue=1, group="-       EchoFrost Frost", parameterName="• Frost - DFAO - Intensity", sortPriority=226}
  }
  MaterialExpressionScalarParameter_96 {
   class=MaterialExpressionScalarParameter
   properties {group="-       EchoFrost Frost", parameterName="• Frost - DFAO - Bias", sortPriority=225}
  }
  MaterialExpressionSetMaterialAttributes_0 {class=MaterialExpressionSetMaterialAttributes, inputs [MaterialAttributes {from="", output=""}]}
  MaterialExpressionSetMaterialAttributes_2 {
   class=MaterialExpressionSetMaterialAttributes
   inputs [
    MaterialAttributes {from="", output=""}
    BaseColor {from=MaterialExpressionPower_1, output=""}
   ]
   properties {attributeSetTypes [#6]}
  }
  MaterialExpressionSetMaterialAttributes_3 {
   class=MaterialExpressionSetMaterialAttributes
   inputs [
    MaterialAttributes {from=MaterialExpressionNamedRerouteUsage_4, output=""}
    EmissiveColor {from=MaterialExpressionLinearInterpolate_0, output=""}
   ]
   properties {attributeSetTypes [#1]}
  }
  MaterialExpressionSetMaterialAttributes_6 {
   class=MaterialExpressionSetMaterialAttributes
   inputs [
    MaterialAttributes {from="", output=""}
    BaseColor {from=MaterialExpressionVectorParameter_16, output=RGB}
    Specular {from=MaterialExpressionScalarParameter_46, output=""}
    Roughness {from=MaterialExpressionScalarParameter_47, output=""}
    Opacity {from=MaterialExpressionScalarParameter_48, output=""}
    Normal {from=MaterialExpressionReroute_4, output=""}
    SubsurfaceColor {from=MaterialExpressionMultiply_3, output=""}
    EmissiveColor {from=MaterialExpressionConstant_6, output=""}
    OpacityMask {from=MaterialExpressionReroute_34, output=""}
   ]
   properties {attributeSetTypes [
     #6
     #2
     #3
     #7
     #4
     #8
     #1
     #5
    ]}
  }
  MaterialExpressionStaticBoolParameter_0 {
   class=MaterialExpressionStaticBoolParameter
   inputs []
   properties {group="-       EchoFrost Frost", parameterName="• EchoFrost - Frost - World Z Normal - Input Uses Tangent Space Normals", sortPriority=110}
  }
  MaterialExpressionStaticBoolParameter_11 {
   class=MaterialExpressionStaticBoolParameter
   inputs []
   properties {
    desc="T= VTX Normals only\r\nF= Combined Material Normals"
    group="-       EchoFrost Frost"
    parameterName="• EchoFrost - Frost - World Z Normal - Use VTX Normals"
    sortPriority=110
   }
  }
  MaterialExpressionStaticBoolParameter_25 {
   class=MaterialExpressionStaticBoolParameter
   inputs []
   properties {defaultValue=true, group="-       EchoFrost Frost", parameterName="• Frost - DFAO - Force Baked AO on Nanite", sortPriority=210}
  }
  MaterialExpressionStaticBoolParameter_31 {
   class=MaterialExpressionStaticBoolParameter
   inputs []
   properties {defaultValue=true, group="-       EchoFrost Frost", parameterName="• Frost --- Use WORLD Z NORMAL MASKING ---", sortPriority=100}
  }
  MaterialExpressionStaticBoolParameter_42 {
   class=MaterialExpressionStaticBoolParameter
   inputs []
   properties {group="-       EchoFrost Frost", parameterName="• Frost --- Use AO/CURVATURE/THICKNESS MASKING ---", sortPriority=150}
  }
  MaterialExpressionStaticBoolParameter_43 {
   class=MaterialExpressionStaticBoolParameter
   inputs []
   properties {group="-       EchoFrost Frost", parameterName="• Frost --- Use DFAO MASKING ---", sortPriority=200}
  }
  MaterialExpressionStaticBool_0 {class=MaterialExpressionStaticBool}
  MaterialExpressionStaticSwitch_0 {
   class=MaterialExpressionStaticSwitch
   inputs [
    True {from=MaterialExpressionTransform_0, output=""}
    False {from=MaterialExpressionGetMaterialAttributes_16, output=Normal}
    Value {from=MaterialExpressionNamedRerouteUsage_9, output=""}
   ]
   properties {}
  }
  MaterialExpressionStaticSwitch_1 {
   class=MaterialExpressionStaticSwitch
   inputs [
    True {from=MaterialExpressionIf_5, output=""}
    False {from=MaterialExpressionReroute_228, output=""}
    Value {from=MaterialExpressionNamedRerouteUsage_85, output=""}
   ]
   properties {}
  }
  MaterialExpressionStaticSwitch_2 {
   class=MaterialExpressionStaticSwitch
   inputs [
    True {from=MaterialExpressionIf_3, output=""}
    False {from=MaterialExpressionReroute_230, output=""}
    Value {from=MaterialExpressionNamedRerouteUsage_15, output=""}
   ]
   properties {}
  }
  MaterialExpressionStaticSwitch_3 {
   class=MaterialExpressionStaticSwitch
   inputs [
    True {from=MaterialExpressionNamedRerouteUsage_0, output=""}
    False {from=MaterialExpressionConstant_3, output=""}
    Value {from=MaterialExpressionNamedRerouteUsage_2, output=""}
   ]
   properties {}
  }
  MaterialExpressionStaticSwitch_4 {
   class=MaterialExpressionStaticSwitch
   inputs [
    True {from=MaterialExpressionBlendMaterialAttributes_0, output=""}
    False {from=MaterialExpressionNamedRerouteUsage_4, output=""}
    Value {from=MaterialExpressionFunctionInput_1, output=""}
   ]
   properties {}
  }
  MaterialExpressionStaticSwitch_6 {
   class=MaterialExpressionStaticSwitch
   inputs [
    True {from=MaterialExpressionSaturate_1, output=""}
    False {from=MaterialExpressionSaturate_17, output=""}
    Value {from=MaterialExpressionNamedRerouteUsage_12, output=""}
   ]
   properties {desc="T= VTX Normals only\r\nF= Combined Material Normals"}
  }
  MaterialExpressionStaticSwitch_7 {
   class=MaterialExpressionStaticSwitch
   inputs [
    True {from=MaterialExpressionNamedRerouteUsage_7, output=""}
    False {from=MaterialExpressionNamedRerouteUsage_6, output=""}
    Value {from=MaterialExpressionNamedRerouteUsage_46, output=""}
   ]
   properties {desc="** DF AO is for Epic Settings Only!\r\nTrue = AO Baked from VTX or Texture\r\nFalse = AO Distance Field"}
  }
  MaterialExpressionTextureObjectParameter_0 {
   class=MaterialExpressionTextureObjectParameter
   inputs [
    UVs {from="", output=""}
    Apply View MipBias {from="", output=""}
   ]
   properties {
    group="-       EchoFrost Frost"
    parameterName=* Frost - Height Texture
    samplerType=SAMPLERTYPE_Grayscale
    texture=/Game/Environments/Apollo/Sets/Terrain/Textures/T_Apollo_SnowMask
   }
  }
  MaterialExpressionTextureSample_13 {
   class=MaterialExpressionTextureSample
   inputs [
    UVs {from=MaterialExpressionMaterialFunctionCall_30, output=UVs}
    Tex {from=MaterialExpressionTextureObjectParameter_0}
    "DDX(UVs)" {from=MaterialExpressionMaterialFunctionCall_30, output=UVs}
    "DDY(UVs)" {from=MaterialExpressionMaterialFunctionCall_30, output=UVs}
    Apply View MipBias {from=""}
   ]
   properties {mipValueMode=TMVM_Derivative, samplerSource=SSM_Wrap_WorldGroupSettings}
  }
  MaterialExpressionTransform_0 {class=MaterialExpressionTransform, inputs [None {from=MaterialExpressionGetMaterialAttributes_16, output=Normal}]}
  MaterialExpressionTransform_1 {class=MaterialExpressionTransform, inputs [None {from=MaterialExpressionConstant3Vector_1}]}
  MaterialExpressionVectorParameter_0 {
   class=MaterialExpressionVectorParameter
   properties {
    group="-       EchoFrost Frost"
    parameterName="• Frost - AO/Curvature/Thickness - Channel Weights"
    sortPriority=160
    channelNames {a="", b=Thickness, g=Curvature, r=AO}
    defaultValue=<rgba 1 0 0 0>
   }
  }
  MaterialExpressionVectorParameter_16 {
   class=MaterialExpressionVectorParameter
   properties {group="-       EchoFrost Frost", parameterName=* Frost - Color, defaultValue=<rgba 0.42 0.4623 0.5104 1>}
  }
  MaterialExpressionVertexInterpolator_5 {class=MaterialExpressionVertexInterpolator, inputs [VS {from=MaterialExpressionSaturate_33, output=""}], properties {}}
  MaterialExpressionVertexNormalWS_0 {class=MaterialExpressionVertexNormalWS, inputs [], properties {}}
 }}
